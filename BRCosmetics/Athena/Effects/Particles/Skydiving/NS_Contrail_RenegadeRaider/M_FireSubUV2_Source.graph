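# Translucent unlit fire material for Niagara sprites and mesh particles, filed under the Renegade Raider skydiving contrail.
# Emissive combines a masked 8x8 fire sheet, particle color and time-of-day scaling, then passes through an adjustment function.
# Opacity erodes the fire mask with panning noise, scaled by frame grid and driven by dynamic parameters.
# Noise also varies depth-fade distance; mobile skips depth fade, and a static switch offers near-camera fade.
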Material /BRCosmetics/Athena/Effects/Particles/Skydiving/NS_Contrail_RenegadeRaider/M_FireSubUV2_Source
graph {
 properties {
  bUsedWithNiagaraMeshParticles=true
  bUsedWithNiagaraSprites=true
  blendMode=BLEND_Translucent
  pixelDepthOffsetMode=PDOM_Legacy
  shadingModel=MSM_Unlit
  usedShadingModels=Unlit
 }
 nodes [
  n1 = ScalarParameter() {defaultValue=1, group=Time of Day, parameterName=Color Value Morning}
  n2 = ScalarParameter() {defaultValue=1, group=Time of Day, parameterName=Color Value Day}
  n3 = ScalarParameter() {defaultValue=1, group=Time of Day, parameterName=Color Value Evening}
  n4 = ScalarParameter() {defaultValue=1, group=Time of Day, parameterName=Color Value Night}
  n5 = ParticleColor()
  n6 = TextureSampleParameter2D(UVs=TextureCoordinate()) {parameterName=SubUV_Texture, texture=/Game/Effects/Fort_Effects/Textures/Fire/T_FireFluid_1_8X8}
  n7 = ChannelMaskParameter(None=@n6.RGBA) {maskChannel=Alpha, parameterName=Alpha Mask, defaultValue=<rgba 0 0 0 1>}
  n8 = Multiply(A=@n6.R, B=@n7.R)
  MaterialFunctionCall(Emissive=Multiply(
    A=MaterialFunctionCall(Value_Morning=@n1, Value_Day=@n2, Value_Evening=@n3, Value_Night=@n4) {materialFunction=&Scalar_Time_of_Day}.Result
    B=Multiply(A=@n5.RGB, B=@n8)
   )) {materialFunction=&MF_FX_EmissiveAdjustments}
  n9 = ScalarParameter() {defaultValue=1, parameterName=Test X}
  Multiply(A=@n9)
  ParticleMacroUV()
  n10 = ScalarParameter() {defaultValue=60, parameterName=DepthFade_Low}
  n11 = ScalarParameter() {defaultValue=120, parameterName=DepthFade_High}
  n12 = TextureCoordinate()
  n13 = Multiply(A=ScalarParameter() {defaultValue=8, parameterName=FramesX}, B=ComponentMask(None=@n12) {r=true})
  n14 = Multiply(A=ComponentMask(None=@n12) {g=true}, B=ScalarParameter() {defaultValue=8, parameterName=FramesY})
  n15 = Add(A=@n9) {constB=2}
  n16 = Multiply(A=DynamicParameter() {paramNames [LerpNoise,X,PowerErode,Scale]}.Scale, B=ScalarParameter() {defaultValue=1, parameterName=Test Scale})
  n17 = Multiply(
   A=Add(A=AppendVector(A=@n13, B=@n14), B=AppendVector(A=@n9, B=@n15))
   B=LinearInterpolate(A=@n16, B=Floor(None=@n16), Alpha=ScalarParameter() {defaultValue=1, parameterName=FloorScaleParameter})
  ) {desc=Scale}
  n18 = ComponentMask(None=VectorParameter() {parameterName=Noise Panner Speed, defaultValue=<rgba 0 -0.125 0 0>}.RGB) {g=true, r=true}
  n19 = ChannelMaskParameter(None=TextureSampleParameter2D(UVs=Panner(Coordinate=@n17, Speed=@n18) {speedY=-0.125}) {parameterName=NoiseTexture, texture=/Game/Effects/Fort_Effects/Textures/Tile/T_Inky_Smoke_Tile2}.RGBA) {maskChannel=Green, parameterName=NoiseChannelSwitch, defaultValue=<rgba 0 1 0 0>}
  n20 = Multiply(A=@n7.R, B=@n19.R)
  n21 = Multiply(
   A=DynamicParameter() {paramNames [LerpNoise,X,PowerErode,Scale]}.PowerErode
   B=ScalarParameter() {defaultValue=1, parameterName=Test PowerErode}
  )
  n22 = DynamicParameter() {paramNames [LerpNoise,X,PowerErode,Scale]}
  n23 = ScalarParameter() {defaultValue=1, parameterName=Test LerpNoise}
  n24 = Multiply(A=@n5.A, B=LinearInterpolate(A=@n7.R, B=Power(Base=@n20, Exp=@n21), Alpha=Multiply(A=@n22.LerpNoise, B=@n23)))
  n25 = LinearInterpolate(A=@n24, B=Clamp(None=@n24), Alpha=ScalarParameter() {parameterName=Use_Clamp_OnParticleAlpha})
  n26 = ShadingPathSwitch(
   Default=DepthFade(Opacity=@n25, FadeDistance=LinearInterpolate(A=@n10, B=@n11, Alpha=@n19.R) {constA=60, constB=120, desc=Bias Noise})
   Mobile=@n25
  ) {desc="Added pref for mobile,Yang."}
  n27 = ScalarParameter() {defaultValue=150, parameterName=Near Camera Distance}
  StaticSwitchParameter(
   True=Multiply(A=@n26, B=MaterialFunctionCall(Distance=@n27) {materialFunction=&Near_Camera_Fade}.Result)
   False=@n26
  ) {parameterName=Use Near Camera Fade}
  TextureSampleParameterSubUV!(UVs, Apply View MipBias) {parameterName=FireSubUV, texture=&T_FireFluid_1_8X8}
 ]
}
