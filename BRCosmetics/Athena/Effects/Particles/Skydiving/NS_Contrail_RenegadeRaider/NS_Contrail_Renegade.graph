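# Niagara skydiving contrail for the Renegade Raider cosmetic, producing fiery sprite effects on the character's body and limbs.
# System diving attributes supply a Dive value that scales GPU Drops spawned at limb bones, with orange-to-red fading and SubUV animation.
# FlameDecals flow along partitioned bones via FN_BoneFlow, and a front-end preview flag overrides both spawn rates.
# Several other emitters, including mesh body, circle flames and attached ribbons, are present but disabled.
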NiagaraSystem /BRCosmetics/Athena/Effects/Particles/Skydiving/NS_Contrail_RenegadeRaider/NS_Contrail_Renegade
dependencies {
 usedDataInterfaces [
  &NiagaraDataInterfaceCurve
  &NiagaraDataInterfaceColorCurve
  &NiagaraDataInterfaceVectorCurve
  &NiagaraDataInterfaceCurlNoise
 ]
 usedDynamicInputs []
 usedModules [
  &SystemState
  /Niagara/Modules/System/SystemLifeCycle
  /Game/Effects/Niagara/Modules/Skydiving/FN_DivingAttributes/v4/FN_DivingAttributes_v4
 ]
}
emitters {
 Blue_Main_Body {
  bEnabled=false
  emitterUpdateScript [
   &EmitterState(Life Cycle Mode=::NewEnumerator1, Loop Duration=1, UseLoopDelay=false)
   &SpawnRate(SpawnRate=5, Use Spawn Probability=false)
   &SpawnRate(SpawnGroup=1, SpawnRate=5, Use Spawn Probability=false) {module=SpawnRate001}
  ]
  particleSpawnScript [
   /Game/Effects/Niagara/Modules/Skydiving/FN_DivingAttributes/FN_DivingAttributes_v2(Accumulate Dive=0.1, Accumulate Rest=-0.1, SkeletalMesh:NiagaraDataInterfaceSkeletalMesh=$User.SkeletalMesh)
   /SetVariables(Particles.DiveCLAMP:NiagaraFloat=&ClampFloat(Float=&FN_FrontEndFloatModifier(Gameplay=$Particles.Dive, Front End=&Multiply_Float(A=$Particles.Dive, B=10)))) {isSetParameters=true, module=SetVariables_#1}
   /SetVariables(Particles.MaterialRandom:NiagaraFloat=&UniformRangedFloat(Minimum=0, Maximum=1)) {isSetParameters=true, module=SetVariables_#2}
   /Niagara/Modules/Spawn/Initialization/InitializeParticle(
    Color=&LerpLinearColors(
     StartColor=<rgba 50 0.8333 0 0>
     EndColor=&LerpLinearColors(
      StartColor=<rgba 1 0.0167 0 0.1>
      EndColor=<rgba 50 4.1667 0 0.1>
      LerpFactor=&UniformAOrBFloat(A=0.5, B=0.7)
     )
     LerpFactor=$Particles.DiveCLAMP
    )
    Lifetime=&Lerp_Float(
     A=&UniformRangedFloat(Minimum=0.5, Maximum=0.7)
     B=&UniformRangedFloat(Minimum=0.8, Maximum=1)
     Alpha=$Particles.DiveCLAMP
    )
    Material Random:NiagaraFloat=$Particles.MaterialRandom
    Mesh Scale=&Multiply_Vector(A=<0.755 0.32 0.32>, B=&Lerp_Vector(A=<1 1 1>, B=<2 1 1>, Alpha=$Particles.DiveCLAMP))
    Module.Write Color=true
    Module.Write Lifetime=true
    Module.Write Mass=true
    Module.Write Material Random=true
    Module.Write Position=false
    Module.Write Scale=true
    Module.Write SpriteRotation=true
    Module.Write SpriteSize=false
    Sprite Rotation=&UniformRangedFloat(Minimum=-0.2, Maximum=0.2)
    Sprite Size=<5 5>
   )
   &SkeletalMeshLocation(
    Apply To Particle Position=false
    Bone / Socket Index=6
    Bone Sampling Mode=::NewEnumerator4
    Inherit Velocity=false
    MaskSpawnGroup=false
    Sampled Position Offset=<0 5 0>
    Skeletal Mesh=$User.SkeletalMesh
   ) {enabled=false, module=SkeletalMeshLocation001}
   /Niagara/Modules/Spawn/Orientation/V2/InitialMeshOrientation(
    Orientation Vector=<0 0 -1>
    Rotation=&UniformRangedVector(Minimum=<0 0 0>, Maximum=<1 0 0>)
    Use Orientation Vector=true
    Use Rotation Vector=true
   )
  ]
  particleUpdateScript [
   /SetVariables(Particles.DiveCLAMP:NiagaraFloat=&ClampFloat(Float=&FN_FrontEndFloatModifier(Gameplay=$Particles.Dive, Front End=&Multiply_Float(A=$Particles.Dive, B=10)))) {isSetParameters=true, module=SetVariables_#3}
   &ParticleState()
   &SkeletalMeshLocation(
    Bone / Socket Index=6
    Bone Sampling Mode=::NewEnumerator6
    Inherit Velocity=false
    MaskSpawnGroup=false
    Sampled Position Offset=<0 8 0>
    Sampled Position Offset Coordinate Space:ENiagaraMeshTransforms=::NewEnumerator2
    Skeletal Mesh=$User.SkeletalMesh
   )
   /Niagara/Modules/Update/Orientation/V2/UpdateMeshOrientation(
    Delta Time:NiagaraFloat=&Multiply_Float(A=$Engine.DeltaTime, B=&Multiply_Float(A=5, B=$Particles.NormalizedAge))
    Rotation Rate:NiagaraFloat=$Particles.Initial.SpriteRotation
    Rotation Vector:Vector3f=<0 0 1>
   ) {enabled=false}
   &ScaleMeshSize(Scale Factor=&Multiply_Vector(
     A=&Lerp_Vector(
      A=&MakeVector(
       X=&FloatFromCurve(FloatCurve=DataInterface() {propertyValues {Curve {keys [<at 0 0.1>,<at 1 2>]}}})
       Y=&FloatFromCurve(FloatCurve=DataInterface() {propertyValues {Curve {keys [
            <at 0 1 RCIM_Cubic>
            {arriveTangent=-1.1222, interpMode=RCIM_Cubic, leaveTangent=-2.1367, tangentMode=RCTM_User, time=0.9, value=0.5}
            <at 1 0.05>
           ]}}})
       Z=&FloatFromCurve(FloatCurve=DataInterface() {propertyValues {Curve {keys [
            <at 0 1 RCIM_Cubic>
            {arriveTangent=-1.1222, interpMode=RCIM_Cubic, leaveTangent=-2.1367, tangentMode=RCTM_User, time=0.9, value=0.5}
            <at 1 0.05>
           ]}}})
      )
      B=&MakeVector(
       X=&FloatFromCurve(FloatCurve=DataInterface() {propertyValues {Curve {keys [<at 0 1>,<at 1 2>]}}})
       Y=&FloatFromCurve(FloatCurve=DataInterface() {propertyValues {Curve {keys [
            {arriveTangent=-1.5003, interpMode=RCIM_Cubic, leaveTangent=-0.2737, tangentMode=RCTM_User, time=0, value=1}
            {arriveTangent=-2.2772, interpMode=RCIM_Cubic, leaveTangent=-2.2772, tangentMode=RCTM_User, time=1, value=0}
           ]}}})
       Z=&FloatFromCurve(FloatCurve=DataInterface() {propertyValues {Curve {keys [
            {arriveTangent=-1.5003, interpMode=RCIM_Cubic, leaveTangent=-0.2737, tangentMode=RCTM_User, time=0, value=1}
            {arriveTangent=-2.2772, interpMode=RCIM_Cubic, leaveTangent=-2.2772, tangentMode=RCTM_User, time=1, value=0}
           ]}}})
      )
      Alpha=$Particles.DiveCLAMP
     )
     B=&VectorFromFloat(Value=&FloatFromCurve(FloatCurve=DataInterface() {propertyValues {Curve {keys [
           <at 0 0 RCIM_Constant>
           {arriveTangent=5.5862, interpMode=RCIM_Cubic, leaveTangent=5.5862, tangentMode=RCTM_User, time=0.3, value=0}
           <at 0.7 1>
           <at 1 1>
          ]}}}))
    )) {module=ScaleMeshSize001}
   &ScaleColor(
    Scale Alpha=&FloatFromCurve(FloatCurve=DataInterface() {propertyValues {Curve {keys [
         {arriveTangent=8.7245, interpMode=RCIM_Cubic, leaveTangent=8.7245, tangentMode=RCTM_User, time=0, value=0}
         <at 0.45 1 RCIM_Cubic>
         {arriveTangent=-0.5281, interpMode=RCIM_Cubic, leaveTangent=-0.5281, tangentMode=RCTM_User, time=1, value=0}
        ]}}})
    ScaleA=true
    ScaleRGB=true
   )
   &DynamicMaterialParameters(
    Index 0 Param 1=&Lerp_Float(A=0, B=0.5, Alpha=$Particles.DiveCLAMP)
    Module.Param0WriteEnabled=true
    Module.Param1WriteEnabled=false
    Module.Param2WriteEnabled=false
    Module.Param3WriteEnabled=false
   )
  ]
  rendererClasses [&NiagaraMeshRendererProperties]
  renderers [{rendererClass=&NiagaraMeshRendererProperties, rendererIndex=0}]
 }
 CircleFlames {
  bEnabled=false
  emitterSpawnScript [
   &FN_PreserveSkydiveUserVariables() {module=PreserveSkydiveUserVariables}
   /Game/Effects/Niagara/Modules/Partition/FN_PartitionAttributes(Particles Per Partition=10, Partitions=4)
  ]
  emitterUpdateScript [
   &EmitterState()
   &SpawnRate(SpawnRate=20, Use Spawn Probability=false)
  ]
  particleSpawnScript [
   /Game/Effects/Niagara/Modules/Partition/FN_ParticlePartition_v2/FN_Particle_Partition(Pattern:ENiagaraPartitionWaveForm=::NewEnumerator1, Seed:NiagaraInt32=$Emitter.RandomSeed)
   &InitializeParticle(
    Color=/Niagara/DynamicInputs/TypeConversions/MakeLinearColorFromFloat(Value:NiagaraFloat=&UniformRangedFloat(Minimum=2, Maximum=5))
    Lifetime=&UniformRangedFloat(Minimum=0.4, Maximum=0.8)
    Module.Write Color=true
    Module.Write Lifetime=true
    Module.Write Mass=false
    Module.Write Material Random=true
    Module.Write Position=false
    Module.Write Scale=false
    Module.Write SpriteRotation=true
    Module.Write SpriteSize=true
    Position=&SimulationPosition()
    Sprite Rotation=&UniformRangedFloat(Minimum=0, Maximum=360)
    Sprite Size=&Vector2DFromFloat(Value=&UniformRangedFloat(Minimum=10, Maximum=15))
   )
   /SetVariables(Particles.LightRadius=30) {isSetParameters=true, module=SetVariables_#4}
   /SetVariables(Particles.BoneFlowAngle:NiagaraFloat=&UniformRangedFloat(Minimum=0, Maximum=6.4)) {isSetParameters=true, module=SetVariables_#5}
  ]
  particleUpdateScript [
   &SampleSkeletalMeshSkeleton(Mesh:NiagaraDataInterfaceSkeletalMesh=$User.SkeletalMesh, Specific Bone Index:NiagaraInt32=$Particles.SpawnGroup)
   &Color(Color=&ColorFromCurve(ColorCurve=DataInterface() {propertyValues {
       AlphaCurve {keys [<at -0.0351 0>, {arriveTangent=0.9661, interpMode=RCIM_Cubic, leaveTangent=0.9661, time=0.094, value=1}, <at 1 1 RCIM_Cubic>]}
       BlueCurve {keys [<at -0.0023 2 RCIM_Cubic>,<at 0.9657 2>]}
       GreenCurve {keys [<at -0.0023 2.6 RCIM_Cubic>,<at 0.9657 3.5>]}
       RedCurve {keys [<at -0.0023 20 RCIM_Cubic>,<at 0.9657 20>]}
      }})) {enabled=false}
   &ScaleColor(
    Scale RGB=&MakeVectorFromLinearColorRGB(Color=&ColorFromCurve(ColorCurve=DataInterface() {propertyValues {
        AlphaCurve {keys [<at 0 1 RCIM_Cubic>,<at 1 1 RCIM_Cubic>]}
        BlueCurve {keys [<at 0.1333 0.05 RCIM_Cubic>, {interpMode=RCIM_Cubic, leaveTangent=1.8683, time=0.4267, value=0.25}, <at 0.8284 1>]}
        GreenCurve {keys [<at 0.1333 0.145 RCIM_Cubic>, {interpMode=RCIM_Cubic, leaveTangent=7.9093, time=0.4267, value=0.5667}, <at 0.8284 4.1667>]}
        RedCurve {keys [<at 0.1333 1 RCIM_Cubic>, {interpMode=RCIM_Cubic, leaveTangent=37.3667, time=0.4267, value=5}, <at 0.8284 20>]}
       }}))
    ScaleA=true
    ScaleRGB=true
   )
   &ParticleState()
   &FN_BoneFlow(
    Angle:NiagaraFloat=$Particles.BoneFlowAngle
    BoneID:NiagaraInt32=&Add_Integer(A=$Particles.Partition.ID, B=0)
    Height:NiagaraFloat=&FloatFromCurve(FloatCurve=DataInterface() {propertyValues {Curve {keys [
         {arriveTangent=2.2164, interpMode=RCIM_Cubic, leaveTangent=2.2164, tangentMode=RCTM_User, time=0, value=0.4}
         {arriveTangent=0.7096, interpMode=RCIM_Cubic, leaveTangent=0.7096, tangentMode=RCTM_User, time=0.9897, value=1.2}
        ]}}})
    Interpolate Mesh Sampling=false
    Radius:NiagaraFloat=&FloatFromCurve(
     FloatCurve=DataInterface() {propertyValues {Curve {keys [
         <at 0.23 0.2 RCIM_Cubic>
         {arriveTangent=1.3298, interpMode=RCIM_Cubic, leaveTangent=1.3298, time=0.6661, value=0.6553}
         {arriveTangent=-2.8131, interpMode=RCIM_Cubic, leaveTangent=-2.8131, tangentMode=RCTM_User, time=0.863, value=1.0417}
        ]}}}
     Scale Curve=5
    )
    Set Material Dynamic=true
    SkeletalMesh:NiagaraDataInterfaceSkeletalMesh=$User.SkeletalMesh
    Write Position Attribute=true
   )
   &CameraOffset(Camera Offset Amount=&UniformRangedFloat(Minimum=1, Maximum=3))
   &ScaleSpriteSize(Scale Factor=&Vector2DFromFloat(Value=&FloatFromCurve(FloatCurve=DataInterface() {propertyValues {Curve {keys [
          {arriveTangent=0.3706, interpMode=RCIM_Cubic, leaveTangent=0.3706, tangentMode=RCTM_User, time=-0.0037, value=0.1155}
          {arriveTangent=1.1247, interpMode=RCIM_Cubic, leaveTangent=1.1247, tangentMode=RCTM_User, time=0.1272, value=0.6973}
          <at 0.7695 0.5625>
         ]}}})))
   &SpriteRotationRate(Rotation Rate=&UniformRangedFloat(Minimum=-300, Maximum=300))
   &DynamicMaterialParameters(
    Index 2 Param 1=&FloatFromCurve(FloatCurve=DataInterface() {propertyValues {Curve {keys [
         {arriveTangent=0.8618, interpMode=RCIM_Cubic, leaveTangent=0.8618, tangentMode=RCTM_User, time=0, value=0.0773}
         {arriveTangent=3.7373, interpMode=RCIM_Cubic, leaveTangent=3.7373, tangentMode=RCTM_User, time=0.5779, value=1.0908}
        ]}}})
    Module.Param0WriteEnabled2=true
    Module.Param1WriteEnabled2=true
    Module.Param2WriteEnabled2=true
    Module.Param3WriteEnabled2=true
    Write Parameter Index 0=false
    Write Parameter Index 2=true
   )
  ]
  rendererClasses [&NiagaraSpriteRendererProperties]
  renderers [{rendererClass=&NiagaraSpriteRendererProperties}]
 }
 Drops {
  simTarget=GPUComputeSim
  emitterUpdateScript [
   &EmitterState(Inactive Response=::NewEnumerator1, Life Cycle Mode=::NewEnumerator1, Loop Behavior=::NewEnumerator1, Loop Duration=3600, UseLoopDelay=false)
   &SpawnRate(
    SpawnGroup:NiagaraInt32=&RandomRangeInt(Minimum=0, Maximum=1)
    SpawnRate=&MakeCustomFloatFromBool(Bool=$User.bIsFrontEndPreview, True Float=200, False Float=&Lerp_Float(A=0, B=100, Alpha=$System.Dive))
    Use Spawn Probability=false
   )
  ]
  particleSpawnScript [
   &FN_SetSpawnGroup()
   /Niagara/Modules/Spawn/Initialization/V2/InitializeParticle(
    Color=<rgba 40 40 40 1>
    Lifetime Max=0.3
    Lifetime Min=0.2
    Lifetime Mode=::NewEnumerator1
    Sprite Size Mode=::NewEnumerator3
    Uniform Sprite Size=1
    UsePositionOffset=false
    Write Material Random=true
   ) {module=InitializeParticle001}
   &SkeletalMeshLocation(
    Bone / Socket Index=$Particles.SpawnGroup
    Bone Sampling Mode=::NewEnumerator4
    Inherit Velocity=false
    Interpolate Mesh Sampling=false
    MaskSpawnGroup=false
    Sampled Position Offset=&RandomRangeVector(Minimum=<5 5 5>, Maximum=<-5 -5 -5>)
    Sampled Position Offset Coordinate Space:ENiagaraMeshTransforms=::NewEnumerator3
    Skeletal Mesh=$User.SkelMeshLimbs
   )
   /SetVariables(Particles.Velocity=&UniformRangedVector(Minimum=<-100 -100 200>, Maximum=<100 100 300>)) {enabled=false, isSetParameters=true}
   /SetVariables(Particles.SpriteSize=&MakeCustomVector2DFromBool(
     Bool=$User.bIsFrontEndPreview
     True Vector2D=&RandomRangeVector2D(Minimum=<15 15>, Maximum=<25 25>)
     False Vector2D=&Multiply_Vector2(
      A=<5 12>
      B=&Vector2DFromFloat(Value=&UniformRangedFloat(Minimum=3, Maximum=5))
     )
    )) {isSetParameters=true}
   &AddVelocityInCone(Cone Angle=10, Cone Axis=<0 0 1>, Module.Use Velocity Falloff On Cone Axis=false, Velocity Strength=500)
   &SphereLocation(
    MaskSpawnGroup=false
    Override Seed=false
    Sphere Radius=1
    Surface Only=false
    UseNonUniformScale=false
    UseOffset=false
    UseSphereOrientation=false
   )
  ]
  particleUpdateScript [
   &ParticleState()
   &GravityForce(Gravity=<0 0 -100>)
   &Drag(Use Linear Drag=true, Use Rotational Drag=false)
   &DynamicMaterialParameters(
    Index 0 Param 1=&FloatFromCurve(FloatCurve=DataInterface() {propertyValues {Curve {keys [{arriveTangent=2.125, interpMode=RCIM_Cubic, time=0, value=1.4}, <at 0.4 1 RCIM_Cubic>]}}})
    Index 0 Param 2=1
    Index 0 Param 3=&FloatFromCurve(
     FloatCurve=DataInterface() {propertyValues {Curve {keys [
         {arriveTangent=-0.3008, interpMode=RCIM_Cubic, leaveTangent=-0.3008, tangentMode=RCTM_User, time=0, value=0.2}
         {arriveTangent=4.1145, interpMode=RCIM_Cubic, leaveTangent=4.1145, tangentMode=RCTM_User, time=0.5809, value=1}
        ]}}}
     Scale Curve=3
    )
    Index 0 Param 4=1
    Module.Param0WriteEnabled=true
    Module.Param0WriteEnabled1=true
    Module.Param1WriteEnabled=true
    Module.Param1WriteEnabled1=true
    Module.Param2WriteEnabled=true
    Module.Param2WriteEnabled1=true
    Module.Param3WriteEnabled=true
    Module.Param3WriteEnabled1=true
    Write Parameter Index 1=true
   ) {module=DynamicMaterialParameters001}
   &SolveForcesAndVelocity(Clamp Velocity=false, Limit Acceleration=false)
   &ScaleColor(
    Scale Alpha=&FloatFromCurve(FloatCurve=DataInterface() {propertyValues {Curve {keys [
         <at 0.0993 0.05 RCIM_Cubic>
         {arriveTangent=1.6041, time=0.1493, value=1.05}
         {arriveTangent=-1.249, interpMode=RCIM_Cubic, leaveTangent=-1.249, time=0.75, value=1}
         <at 0.99 0 RCIM_Cubic>
        ]}}})
    Scale RGB=&MakeVectorFromLinearColorRGB(Color=&ColorFromCurve(ColorCurve=DataInterface() {propertyValues {
        AlphaCurve {keys [<at 0 1 RCIM_Cubic>,<at 1 1 RCIM_Cubic>]}
        BlueCurve {keys [
          <at 0.088 0.5>
          {arriveTangent=-0.5511, interpMode=RCIM_Cubic, leaveTangent=-0.5511, time=0.4884, value=0.02}
          <at 0.8692 0.1>
          <at 0.9589 0.02>
         ]}
        GreenCurve {keys [
          <at 0.088 0.9167>
          {arriveTangent=-1.0295, interpMode=RCIM_Cubic, leaveTangent=-1.0295, time=0.4884, value=0.755}
          <at 0.8692 0.1>
          <at 0.9589 0.02>
         ]}
        RedCurve {keys [<at 0.088 1>,<at 0.4884 1 RCIM_Cubic>,<at 0.8692 1>,<at 0.9589 1>]}
       }}))
    ScaleA=true
    ScaleRGB=true
   ) {enabled=false}
   &ScaleColor(
    Scale Alpha=&FloatFromCurve(
     FloatCurve=DataInterface() {propertyValues {Curve {keys [
         {interpMode=RCIM_Cubic, tangentMode=RCTM_User, time=0, value=0}
         {interpMode=RCIM_Cubic, tangentMode=RCTM_User, time=0.5, value=1}
         {tangentMode=RCTM_User, time=1, value=0}
        ]}}}
     Scale Curve=0.2
    )
    Scale RGB=&MakeVectorFromLinearColorRGB(Color=&ColorFromCurve(ColorCurve=DataInterface() {propertyValues {
        AlphaCurve {keys [<at 0 1 RCIM_Cubic>,<at 1 1 RCIM_Cubic>]}
        BlueCurve {keys [<at 0.0287 0.03>,<at 0.0898 0.03>,<at 0.6555 0.03 RCIM_Cubic>]}
        GreenCurve {keys [<at 0.0287 0.6767>,<at 0.0898 0.2887>,<at 0.6555 0.03 RCIM_Cubic>]}
        RedCurve {keys [<at 0.0287 1>,<at 0.0898 1>,<at 0.6555 1 RCIM_Cubic>]}
       }}))
    ScaleA=true
    ScaleRGB=true
   ) {module=ScaleColor001}
   &ScaleSpriteSize(Scale Factor=&MakeVector2D(
     X=&FloatFromCurve(FloatCurve=DataInterface() {propertyValues {Curve {keys [
          {interpMode=RCIM_Cubic, tangentMode=RCTM_User, time=0, value=1}
          {arriveTangent=-0.7342, interpMode=RCIM_Cubic, leaveTangent=-0.7342, tangentMode=RCTM_User, time=1, value=1}
         ]}}})
     Y=&FloatFromCurve(FloatCurve=DataInterface() {propertyValues {Curve {keys [
          {interpMode=RCIM_Cubic, tangentMode=RCTM_User, time=0, value=2}
          {arriveTangent=-0.1377, interpMode=RCIM_Cubic, leaveTangent=-0.1377, tangentMode=RCTM_User, time=1, value=1}
         ]}}})
    ))
   &SubUVAnimation(End Frame=18, SubUV Animation Mode:ENiagara_SubUVLookupMode=::NewEnumerator2)
  ]
  rendererClasses [&NiagaraSpriteRendererProperties]
  renderers [
   {rendererClass=&NiagaraSpriteRendererProperties}
   {rendererClass=&NiagaraSpriteRendererProperties, rendererIndex=1}
   {rendererClass=&NiagaraSpriteRendererProperties, rendererIndex=2}
  ]
 }
 FN_AttachedRibbon_v2 {
  bEnabled=false
  emitterUpdateScript [
   &EmitterState()
   &FN_SpawnRibbon_v2(
    Number of Ribbons=4
    Spawn Mode=::NewEnumerator0
    SpawnRate:NiagaraFloat=&Lerp_Float(A=0, B=30, Alpha=$System.Dive)
   )
   &FN_Integrator()
  ]
  particleSpawnScript [
   &FN_InitializeRibbon_v3(
    Enable Skeletal Position=true
    Lifetime:NiagaraFloat=&Lerp_Float(A=0.2, B=0.1, Alpha=$System.Dive)
    Ribbon Width:NiagaraFloat=&Lerp_Float(A=20, B=10, Alpha=$System.Dive)
   )
   &FN_FlipRibbonV()
  ]
  particleUpdateScript [
   /Niagara/Modules/Update/Forces/CurlNoiseForce(
    Noise Strength=300
    Pan Noise=false
    Sample Position:Vector3f=/Niagara/DynamicInputs/Transforms/TransformPosition(Position=Linked() {linkedVariable {name=Particles.Position, type=NiagaraPosition}}, Source Space=::World, Destination Space=::Local)
   )
   &AddVelocity(Velocity=&MakeCustomVectorFromBool(
     Bool=$Particles.Locked
     True Vector=<0 0 0>
     False Vector=&FN_FrontEndVector3Modifier(Gameplay=<0 0 500>, Front End=<0 0 500>)
    ))
   &Drag(Use Linear Drag=true, Use Rotational Drag=false) {enabled=false}
   &SolveForcesAndVelocity(Clamp Velocity=false, Limit Acceleration=false)
   &FN_UpdateRibbon(
    Enable Skeletal Position=true
    Initial Color:LinearColor=$Particles.Initial.Color
    Initial RibbonWidth:NiagaraFloat=$Particles.Initial.RibbonWidth
   )
   /Game/Effects/Niagara/Modules/Materials/FN_RibbonMaterial/FN_RibbonMaterial(
    Color:LinearColor=&ColorFromCurve(ColorCurve=DataInterface() {propertyValues {
       AlphaCurve {keys [<at 0 1 RCIM_Cubic>,<at 0.2749 0.5 RCIM_Cubic>]}
       BlueCurve {keys [<at 0 0.2 RCIM_Cubic>,<at 0.0514 0.2>,<at 0.2019 0.0677>]}
       GreenCurve {keys [<at 0 0.2 RCIM_Cubic>,<at 0.0514 0.3333>,<at 0.2019 0.0677>]}
       RedCurve {keys [<at 0 1 RCIM_Cubic>,<at 0.0514 1>,<at 0.2019 0.0677>]}
      }})
    Color Pan U Speed=0.5
    Jitter by Particles MaterialRandom=true
    Macro Distortion Amount:NiagaraFloat=&FloatFromCurve(FloatCurve=DataInterface() {propertyValues {Curve {keys [<at 0.0051 0.2>,<at 0.2115 0.4>,<at 0.4914 0.115>,<at 1 0>]}}}, Scale Curve=0)
    Macro Pan Speed=1
    Opacity Erosion:NiagaraFloat=&FloatFromCurve(FloatCurve=DataInterface() {propertyValues {Curve {keys [
         {interpMode=RCIM_Cubic, tangentMode=RCTM_User, time=0, value=0}
         {arriveTangent=1, interpMode=RCIM_Cubic, leaveTangent=1, tangentMode=RCTM_User, time=0.7307, value=0.1673}
         <at 0.9872 0.5>
        ]}}})
    Opacity Pan U Speed=1
    Scale Color:Vector3f=&Lerp_Vector(A=<0 0 0>, B=<10 5 1>, Alpha=$System.Dive)
    TIme:NiagaraFloat=$Emitter.FN_Integrator.Value
    UV0 Scale=0.5
   ) {enabled=false}
   /Game/Effects/Niagara/Modules/Materials/FN_RibbonMaterial/FN_RibbonMaterial_v2(
    Color:LinearColor=&ColorFromCurve(ColorCurve=DataInterface() {propertyValues {
       AlphaCurve {keys [<at 0 1 RCIM_Cubic>,<at 0.2749 0.5 RCIM_Cubic>]}
       BlueCurve {keys [<at 0 0.2 RCIM_Cubic>,<at 0.0514 0.2>,<at 0.2019 0.0677>]}
       GreenCurve {keys [<at 0 0.2 RCIM_Cubic>,<at 0.0514 0.3333>,<at 0.2019 0.0677>]}
       RedCurve {keys [<at 0 1 RCIM_Cubic>,<at 0.0514 1>,<at 0.2019 0.0677>]}
      }})
    Color Pan U Offset=1
    Color Pan U Speed=0.5
    Macro Distortion Amount:NiagaraFloat=&FloatFromCurve(FloatCurve=DataInterface() {propertyValues {Curve {keys [<at 0.0051 0.2>,<at 0.2115 0.4>,<at 0.4914 0.115>,<at 1 0>]}}}, Scale Curve=0)
    Micro Pan Speed=1
    Opacity Erosion:NiagaraFloat=&FloatFromCurve(FloatCurve=DataInterface() {propertyValues {Curve {keys [
         {interpMode=RCIM_Cubic, tangentMode=RCTM_User, time=0, value=0}
         {arriveTangent=1, interpMode=RCIM_Cubic, leaveTangent=1, tangentMode=RCTM_User, time=0.7307, value=0.1673}
         <at 0.9872 0.5>
        ]}}})
    TIme:NiagaraFloat=$Emitter.FN_Integrator.Value
   )
  ]
  rendererClasses [&NiagaraRibbonRendererProperties]
  renderers [{rendererClass=&NiagaraRibbonRendererProperties}]
 }
 FN_AttachedRibbon_v2001 {
  bEnabled=false
  emitterUpdateScript [
   &EmitterState()
   &FN_SpawnRibbon_v2(
    Number of Ribbons=4
    Spawn Mode=::NewEnumerator0
    SpawnRate:NiagaraFloat=&Lerp_Float(A=0, B=30, Alpha=$System.Dive)
   )
   &FN_Integrator()
  ]
  particleSpawnScript [
   &FN_InitializeRibbon_v3(
    Enable Skeletal Position=true
    Lifetime:NiagaraFloat=&Lerp_Float(A=0.2, B=0.1, Alpha=$System.Dive)
    Ribbon Width:NiagaraFloat=&Lerp_Float(A=20, B=10, Alpha=$System.Dive)
   )
   &FN_FlipRibbonV()
  ]
  particleUpdateScript [
   &CurlNoiseForce(
    Noise Strength=300
    Pan Noise=false
    Sample Position:Vector3f=&TransformPosition(Position=Linked() {linkedVariable {name=Particles.Position, type=NiagaraPosition}}, Source Space=::World, Destination Space=::Local)
   )
   &AddVelocity(Velocity=&MakeCustomVectorFromBool(
     Bool=$Particles.Locked
     True Vector=<0 0 0>
     False Vector=&FN_FrontEndVector3Modifier(Gameplay=<0 0 500>, Front End=<0 0 500>)
    ))
   &Drag(Use Linear Drag=true, Use Rotational Drag=false) {enabled=false}
   &SolveForcesAndVelocity(Clamp Velocity=false, Limit Acceleration=false)
   &FN_UpdateRibbon(
    Enable Skeletal Position=true
    Initial Color:LinearColor=$Particles.Initial.Color
    Initial RibbonWidth:NiagaraFloat=$Particles.Initial.RibbonWidth
   )
   &FN_RibbonMaterial(
    Color:LinearColor=&ColorFromCurve(ColorCurve=DataInterface() {propertyValues {
       AlphaCurve {keys [<at 0 1 RCIM_Cubic>,<at 0.2749 0.5 RCIM_Cubic>]}
       BlueCurve {keys [<at 0 0.2 RCIM_Cubic>,<at 0.0514 0.2>,<at 0.2019 0.0677>]}
       GreenCurve {keys [<at 0 0.2 RCIM_Cubic>,<at 0.0514 0.3333>,<at 0.2019 0.0677>]}
       RedCurve {keys [<at 0 1 RCIM_Cubic>,<at 0.0514 1>,<at 0.2019 0.0677>]}
      }})
    Color Pan U Speed=0.5
    Jitter by Particles MaterialRandom=true
    Macro Distortion Amount:NiagaraFloat=&FloatFromCurve(FloatCurve=DataInterface() {propertyValues {Curve {keys [<at 0.0051 0.2>,<at 0.2115 0.4>,<at 0.4914 0.115>,<at 1 0>]}}}, Scale Curve=0)
    Macro Pan Speed=1
    Opacity Erosion:NiagaraFloat=&FloatFromCurve(FloatCurve=DataInterface() {propertyValues {Curve {keys [
         {interpMode=RCIM_Cubic, tangentMode=RCTM_User, time=0, value=0}
         {arriveTangent=1, interpMode=RCIM_Cubic, leaveTangent=1, tangentMode=RCTM_User, time=0.7307, value=0.1673}
         <at 0.9872 0.5>
        ]}}})
    Opacity Pan U Speed=1
    Scale Color:Vector3f=&Lerp_Vector(A=<0 0 0>, B=<10 5 1>, Alpha=$System.Dive)
    TIme:NiagaraFloat=$Emitter.FN_Integrator.Value
    UV0 Scale=0.5
   ) {enabled=false}
   &FN_RibbonMaterial_v2(
    Color:LinearColor=&ColorFromCurve(ColorCurve=DataInterface() {propertyValues {
       AlphaCurve {keys [<at 0 1 RCIM_Cubic>,<at 0.2749 0.5 RCIM_Cubic>]}
       BlueCurve {keys [<at 0 0.2 RCIM_Cubic>,<at 0.0514 0.2>,<at 0.2019 0.0677>]}
       GreenCurve {keys [<at 0 0.2 RCIM_Cubic>,<at 0.0514 0.3333>,<at 0.2019 0.0677>]}
       RedCurve {keys [<at 0 1 RCIM_Cubic>,<at 0.0514 1>,<at 0.2019 0.0677>]}
      }})
    Color Pan U Offset=1
    Color Pan U Speed=0.5
    Macro Distortion Amount:NiagaraFloat=&FloatFromCurve(FloatCurve=DataInterface() {propertyValues {Curve {keys [<at 0.0051 0.2>,<at 0.2115 0.4>,<at 0.4914 0.115>,<at 1 0>]}}}, Scale Curve=0)
    Micro Pan Speed=1
    Opacity Erosion:NiagaraFloat=&FloatFromCurve(FloatCurve=DataInterface() {propertyValues {Curve {keys [
         {interpMode=RCIM_Cubic, tangentMode=RCTM_User, time=0, value=0}
         {arriveTangent=1, interpMode=RCIM_Cubic, leaveTangent=1, tangentMode=RCTM_User, time=0.7307, value=0.1673}
         <at 0.9872 0.5>
        ]}}})
    TIme:NiagaraFloat=$Emitter.FN_Integrator.Value
   )
  ]
  rendererClasses [&NiagaraRibbonRendererProperties]
  renderers [{rendererClass=&NiagaraRibbonRendererProperties}]
 }
 FN_Contrails_Ribbon_Small001 {
  bEnabled=false
  emitterUpdateScript [
   &EmitterScalability(
    Module.bApplySpawnCountScale=false
    Module.bApplySpawnCountScaleByDistanceFraction=false
    Module.bUseMaxDistance=false
    Module.bUseMinDistance=false
   ) {enabled=false}
   &EmitterLifeCycle(CompleteOnInactive=true) {enabled=false}
   &EmitterState()
   &FN_SpawnRibbon_v2(Number of Ribbons=2, Spawn Mode=::NewEnumerator0)
  ]
  particleSpawnScript [
   &FN_SetSpawnGroup_v2(Number of Spawn Groups=2)
   &FN_InitializeRibbon_v3(Enable Skeletal Position=true, Lifetime=1, Ribbon Width=20)
   &FN_InitializeRibbon_v3(
    Enable Skeletal Position=true
    Lifetime=1
    Ribbon Width=50
    Sampled Position Coordinate Space:ENiagaraCoordinateSpace=::Local
    Sampled Rotation Coordinate Space:ENiagaraCoordinateSpace=::Local
    Sampled Velocity Coordinate Space:ENiagaraCoordinateSpace=::Local
   ) {module=FN_InitializeRibbon_v3001}
   &SkeletalMeshLocation(
    Bone / Socket Index=HlslExpression() {hlslExpression="(Particles.SpawnGroup*3)+2"}
    Bone Sampling Mode=::NewEnumerator4
    Inherit Velocity=false
    "Inherit Velocity (Scale)"=0
    Interpolate Mesh Sampling=false
    MaskSpawnGroup=false
    Sampled Position Offset=<0 5 -10>
    Sampled Position Offset Coordinate Space:ENiagaraMeshTransforms=::NewEnumerator2
    Skeletal Mesh=$User.SkeletalMesh
   )
   &AddVelocity(
    Scale Added Velocity=<5 5 1>
    Velocity=&UniformRangedVector(Minimum=<10 10 -50>, Maximum=<-10 -10 -50>, Randomness Mode=::NewEnumerator1, Random Seed=8)
   )
   &AddVelocity(Velocity=&MakeCustomVectorFromBool(
     Bool=$Particles.Locked
     True Vector=<0 0 0>
     False Vector=&FN_FrontEndVector3Modifier(Gameplay=<0 0 400>, Front End=<0 0 310>)
    )) {module=AddVelocity001}
  ]
  particleUpdateScript [
   &AccelerationForce(
    Acceleration=&VectorFromCurve(VectorCurve=DataInterface() {propertyValues {
       XCurve {keys [<at 0.0017 132.1921>,<at 0.9905 -65.3336>]}
       YCurve {keys [<at -0.0011 1.0067>,<at 0.9988 0>]}
       ZCurve {keys [<at -0.0011 -300>,<at 0.9957 -200.05>]}
      }})
    Coordinate Space=::Local
   ) {enabled=false}
   &Drag(Drag=0.9, Use Linear Drag=true, Use Rotational Drag=false) {enabled=false}
   &InheritVelocity(
    Module.Limit Speed=true
    Velocity Limit=10000
    Velocity Scale=&MakeCustomVectorFromBool(Bool=$Particles.Initial.Locked, True Vector=<0 0 0>, False Vector=<-0.03 -0.03 -0.01>)
    Velocity Source=&TransformVector(Vector=$Engine.Owner.Velocity, Source Space=::World, Destination Space=::Local)
   ) {enabled=false}
   /Game/Effects/Niagara/Modules/Skydiving/FN_DivingAttributes/FN_DivingAttributes() {enabled=false}
   &SolveForcesAndVelocity(Clamp Velocity=false, Limit Acceleration=false)
   &FN_UpdateRibbon(
    Enable Skeletal Position=true
    Initial Color:LinearColor=$Particles.Initial.Color
    Initial RibbonWidth:NiagaraFloat=$Particles.Initial.RibbonWidth
    Sampled Position Coordinate Space:ENiagaraCoordinateSpace=::World
    Sampled Rotation Coordinate Space:ENiagaraCoordinateSpace=::World
    Sampled Velocity Coordinate Space:ENiagaraCoordinateSpace=::World
   )
   &Material(Base Texture Time:NiagaraFloat=&Multiply_Float(A=$Emitter.Age, B=-1))
   &DynamicMaterialParameters(
    Index 0 Param 1=$Particles.Initial.MaterialRandom
    Index 0 Param 2=1
    Module.Param0WriteEnabled=true
    Module.Param1WriteEnabled=true
    Module.Param2WriteEnabled=true
    Module.Param3WriteEnabled=true
   ) {enabled=false}
   &ScaleRibbonWidth(Ribbon Width Scale=&FloatFromCurve(
     FloatCurve=DataInterface() {propertyValues {Curve {keys [
         {arriveTangent=-0.5038, interpMode=RCIM_Cubic, leaveTangent=-0.5038, tangentMode=RCTM_User, time=0.0048, value=0.8}
         <at 0.1993 0.7842>
         {arriveTangent=-0.5043, interpMode=RCIM_Cubic, leaveTangent=-0.5043, tangentMode=RCTM_User, time=0.8528, value=0.9995}
         {arriveTangent=-0.147, interpMode=RCIM_Cubic, leaveTangent=-0.147, tangentMode=RCTM_User, time=1.3081, value=0.4982}
        ]}}}
     Scale Curve=0.65
    ))
   &ScaleColor(
    Scale Alpha=&FloatFromCurve(FloatCurve=DataInterface() {propertyValues {Curve {keys [<at 0 1>,<at 0.05 0.1335>,<at 1 0>]}}})
    Scale RGB=&MakeVectorFromLinearColorRGB(Color=&ColorFromCurve(ColorCurve=DataInterface() {propertyValues {
        AlphaCurve {keys [<at 0 1 RCIM_Cubic>,<at 1 1 RCIM_Cubic>]}
        BlueCurve {keys [<at 0.0366 1>,<at 0.0755 0 RCIM_Cubic>,<at 0.1564 0.0005 RCIM_Cubic>]}
        GreenCurve {keys [<at 0.0366 4.1667>,<at 0.0755 0 RCIM_Cubic>,<at 0.1564 0.0011 RCIM_Cubic>]}
        RedCurve {keys [<at 0.0366 20>,<at 0.0755 1 RCIM_Cubic>,<at 0.1564 0.01 RCIM_Cubic>]}
       }}))
    ScaleA=true
    ScaleRGB=true
   )
  ]
  rendererClasses [&NiagaraRibbonRendererProperties]
  renderers [{rendererClass=&NiagaraRibbonRendererProperties}]
 }
 FN_Contrails_Vertigo_Copied {
  bEnabled=false
  emitterUpdateScript [
   &EmitterScalability(
    Module.bApplySpawnCountScale=false
    Module.bApplySpawnCountScaleByDistanceFraction=false
    Module.bUseMaxDistance=false
    Module.bUseMinDistance=false
   ) {enabled=false}
   &EmitterLifeCycle(CompleteOnInactive=true) {enabled=false}
   &EmitterState()
   &FN_SpawnRibbon_v2(Number of Ribbons=2, Spawn Mode=::NewEnumerator0)
  ]
  particleSpawnScript [
   &FN_SetSpawnGroup_v2(Number of Spawn Groups=2)
   &FN_InitializeRibbon_v3(Enable Skeletal Position=true, Lifetime=2.5, Ribbon Width=60)
   &FN_InitializeRibbon_v3(
    Enable Skeletal Position=true
    Lifetime=2.5
    Ribbon Width=70
    Sampled Position Coordinate Space:ENiagaraCoordinateSpace=::Local
    Sampled Rotation Coordinate Space:ENiagaraCoordinateSpace=::Local
    Sampled Velocity Coordinate Space:ENiagaraCoordinateSpace=::Local
   ) {module=FN_InitializeRibbon_v3001}
   &SkeletalMeshLocation(
    Bone / Socket Index=HlslExpression() {hlslExpression="(Particles.SpawnGroup*3)+2"}
    Bone Sampling Mode=::NewEnumerator4
    Inherit Velocity=false
    "Inherit Velocity (Scale)"=0
    Interpolate Mesh Sampling=false
    MaskSpawnGroup=false
    Sampled Position Offset=<0 5 -10>
    Skeletal Mesh=$User.SkeletalMesh
   )
   &AddVelocity(
    Scale Added Velocity=<5 5 1>
    Velocity=&UniformRangedVector(Minimum=<10 10 -50>, Maximum=<-10 -10 -50>, Randomness Mode=::NewEnumerator1, Random Seed=8)
   )
   &AddVelocity(Velocity=&MakeCustomVectorFromBool(
     Bool=$Particles.Locked
     True Vector=<0 0 0>
     False Vector=&FN_FrontEndVector3Modifier(Gameplay=<0 0 400>, Front End=<0 0 310>)
    )) {module=AddVelocity001}
  ]
  particleUpdateScript [
   &AccelerationForce(
    Acceleration=&VectorFromCurve(VectorCurve=DataInterface() {propertyValues {
       XCurve {keys [<at 0.0017 132.1921>,<at 0.9905 -65.3336>]}
       YCurve {keys [<at -0.0011 1.0067>,<at 0.9988 0>]}
       ZCurve {keys [<at -0.0011 -300>,<at 0.9957 -200.05>]}
      }})
    Coordinate Space=::Local
   ) {enabled=false}
   &Drag(Drag=0.9, Use Linear Drag=true, Use Rotational Drag=false)
   &InheritVelocity(
    Module.Limit Speed=true
    Velocity Limit=10000
    Velocity Scale=&MakeCustomVectorFromBool(Bool=$Particles.Initial.Locked, True Vector=<0 0 0>, False Vector=<-0.03 -0.03 -0.01>)
    Velocity Source=&TransformVector(Vector=$Engine.Owner.Velocity, Source Space=::World, Destination Space=::Local)
   )
   &FN_DivingAttributes() {enabled=false}
   &SolveForcesAndVelocity(Clamp Velocity=false, Limit Acceleration=false)
   &FN_UpdateRibbon(
    Enable Skeletal Position=true
    Initial Color:LinearColor=$Particles.Initial.Color
    Initial RibbonWidth:NiagaraFloat=$Particles.Initial.RibbonWidth
    Sampled Position Coordinate Space:ENiagaraCoordinateSpace=::Local
    Sampled Rotation Coordinate Space:ENiagaraCoordinateSpace=::Local
    Sampled Velocity Coordinate Space:ENiagaraCoordinateSpace=::Local
   )
   &Material(Base Texture Time:NiagaraFloat=&Multiply_Float(A=$Emitter.Age, B=-1))
   &DynamicMaterialParameters(
    Index 0 Param 1=$Particles.Initial.MaterialRandom
    Index 0 Param 2=1
    Module.Param0WriteEnabled=true
    Module.Param1WriteEnabled=true
    Module.Param2WriteEnabled=true
    Module.Param3WriteEnabled=true
   )
   &ScaleRibbonWidth(Ribbon Width Scale=&FloatFromCurve(
     FloatCurve=DataInterface() {propertyValues {Curve {keys [
         {arriveTangent=-0.5038, interpMode=RCIM_Cubic, leaveTangent=-0.5038, tangentMode=RCTM_User, time=0.0048, value=0.8}
         <at 0.1993 0.7842>
         {arriveTangent=-0.5043, interpMode=RCIM_Cubic, leaveTangent=-0.5043, tangentMode=RCTM_User, time=0.8528, value=0.9995}
         {arriveTangent=-0.147, interpMode=RCIM_Cubic, leaveTangent=-0.147, tangentMode=RCTM_User, time=1.3081, value=0.4982}
        ]}}}
     Scale Curve=0.65
    ))
   &Color(
    Color=<rgba 1 1 1 0.85>
    Scale Alpha=&FloatFromCurve(
     FloatCurve=DataInterface() {propertyValues {Curve {keys [
         <at -0.05 0>
         <at 0.01 2>
         <at 0.25 1.5>
         {arriveTangent=-1.8669, interpMode=RCIM_Cubic, leaveTangent=-1.8669, tangentMode=RCTM_User, time=0.3, value=0}
        ]}}}
     Scale Curve=0.8
    )
   )
  ]
  rendererClasses [&NiagaraRibbonRendererProperties]
  renderers [{rendererClass=&NiagaraRibbonRendererProperties}]
 }
 FN_FlameDecals {
  emitterSpawnScript [
   /SetVariables(Emitter.RandomSeed:NiagaraInt32=/Niagara/DynamicInputs/Random/RandomInt(Randomness Mode:ENiagaraRandomnessMode=::NewEnumerator0, Random Seed:NiagaraInt32=0, Fixed Random Seed:NiagaraBool=false)) {isSetParameters=true}
   &FN_PartitionAttributes(Particles Per Partition=10, Partitions=4)
  ]
  emitterUpdateScript [
   &EmitterState(Life Cycle Mode=::NewEnumerator1, Loop Duration=1, UseLoopDelay=false)
   &SpawnRate(
    SpawnRate=&MakeCustomFloatFromBool(Bool=$User.bIsFrontEndPreview, True Float=0, False Float=10)
    Use Spawn Probability=false
   )
  ]
  particleSpawnScript [
   &FN_Particle_Partition(Pattern:ENiagaraPartitionWaveForm=::NewEnumerator1, Seed:NiagaraInt32=$Emitter.RandomSeed)
   /Niagara/Modules/Spawn/Initialization/InitializeParticle(
    Lifetime=2
    Module.Write Color=true
    Module.Write Lifetime=true
    Module.Write Mass=true
    Module.Write Material Random=true
    Module.Write Position=false
    Module.Write Scale=false
    Module.Write SpriteRotation=false
    Module.Write SpriteSize=true
    Position=&SimulationPosition()
    Sprite Size=<5 5>
   )
   /SetVariables(Particles.BoneFlowAngle:NiagaraFloat=&UniformRangedFloat(Minimum=0, Maximum=6.4)) {isSetParameters=true}
  ]
  particleUpdateScript [
   &ParticleState()
   &ScaleSpriteSize(Scale Factor=&Vector2DFromFloat(Value=&FloatFromCurve(
      FloatCurve=DataInterface() {propertyValues {Curve {keys [
          {arriveTangent=0.3706, interpMode=RCIM_Cubic, leaveTangent=0.3706, tangentMode=RCTM_User, time=0, value=0.35}
          {arriveTangent=0.9286, interpMode=RCIM_Cubic, leaveTangent=0.9286, time=0.5, value=0.4}
          {arriveTangent=1.1557, interpMode=RCIM_Cubic, leaveTangent=1.1557, tangentMode=RCTM_User, time=0.7, value=1}
          <at 1 0 RCIM_Cubic>
         ]}}}
      Scale Curve=4
     )))
   &FN_BoneFlow(
    Angle:NiagaraFloat=$Particles.BoneFlowAngle
    BoneID:NiagaraInt32=&Add_Integer(A=$Particles.Partition.ID, B=0)
    Height:NiagaraFloat=&FloatFromCurve(FloatCurve=DataInterface() {propertyValues {Curve {keys [
         {arriveTangent=0.9481, interpMode=RCIM_Cubic, leaveTangent=0.9481, tangentMode=RCTM_User, time=0, value=0.2}
         {arriveTangent=0.3204, interpMode=RCIM_Cubic, leaveTangent=0.3204, tangentMode=RCTM_User, time=0.9897, value=1.1}
        ]}}})
    Interpolate Mesh Sampling=false
    Radius:NiagaraFloat=&FloatFromCurve(
     FloatCurve=DataInterface() {propertyValues {Curve {keys [
         <at 0.1 0.5 RCIM_Cubic>
         {arriveTangent=-0.3642, interpMode=RCIM_Cubic, leaveTangent=-0.3642, time=0.481, value=0.5855}
         {arriveTangent=0.0763, interpMode=RCIM_Cubic, leaveTangent=0.0763, tangentMode=RCTM_User, time=0.8773, value=0.2169}
        ]}}}
     Scale Curve=5
    )
    SkeletalMesh:NiagaraDataInterfaceSkeletalMesh=$User.SkeletalMesh
    Write Position Attribute=true
   )
   &CameraOffset(Camera Offset Amount=10)
   &ScaleColor(
    Scale Alpha=&FloatFromCurve(FloatCurve=DataInterface() {propertyValues {Curve {keys [{interpMode=RCIM_Cubic, tangentMode=RCTM_User, time=0, value=1}, {interpMode=RCIM_Cubic, tangentMode=RCTM_User, time=1, value=0}]}}})
    Scale RGB=&MakeVectorFromLinearColorRGB(Color=&ColorFromCurve(ColorCurve=DataInterface() {propertyValues {
        AlphaCurve {keys [<at 0 1 RCIM_Cubic>,<at 1 1 RCIM_Cubic>]}
        BlueCurve {keys [<at 0.1333 0 RCIM_Cubic>,<at 0.5316 0.05>,<at 0.7452 1 RCIM_Cubic>]}
        GreenCurve {keys [<at 0.1333 0 RCIM_Cubic>,<at 0.5316 0.2083>,<at 0.7452 2.2667 RCIM_Cubic>]}
        RedCurve {keys [<at 0.1333 1 RCIM_Cubic>,<at 0.5316 1>,<at 0.7452 20 RCIM_Cubic>]}
       }}))
    ScaleA=true
    ScaleRGB=true
   )
  ]
  rendererClasses [&NiagaraSpriteRendererProperties]
  renderers [{rendererClass=&NiagaraSpriteRendererProperties}]
 }
 FN_SnowTrail_Ribbon_v4_A {
  bEnabled=false
  emitterUpdateScript [
   &EmitterState()
   &FN_SpawnRibbon_v2(Number of Ribbons=4, Spawn Mode=::NewEnumerator0)
  ]
  particleSpawnScript [
   &FN_InitializeRibbon_v3(Enable Skeletal Position=true, Lifetime=0.7, Ribbon Width=60) {enabled=false}
   &FN_SetSpawnGroup_v2(Override SpawnGroup=true)
   &FN_InitializeRibbon_v3(
    Color:LinearColor=&MakeLinearColorFromFloat(Value:NiagaraFloat=0.3)
    Enable Skeletal Position=true
    Lifetime=0.7
    Ribbon Width=50
    Sampled Position Coordinate Space:ENiagaraCoordinateSpace=::Local
    Sampled Rotation Coordinate Space:ENiagaraCoordinateSpace=::Local
    Sampled Velocity Coordinate Space:ENiagaraCoordinateSpace=::Local
   ) {enabled=false, module=FN_InitializeRibbon_v3001}
   &FN_InitializeRibbon_v3(
    Color:LinearColor=<rgba 0.1 0.1 0.1 5>
    Enable Skeletal Position=true
    Lifetime=0.7
    Ribbon Width:NiagaraFloat=&Lerp_Float(A=20, B=20, Alpha=$System.Dive)
   ) {module=FN_InitializeRibbon_v3002}
   &AddVelocity(Velocity=<0 0 410>) {enabled=false}
   &AddVelocity(Velocity=&MakeCustomVectorFromBool(
     Bool=$Particles.Locked
     True Vector=<0 0 0>
     False Vector=&FN_FrontEndVector3Modifier(Gameplay=<0 0 500>, Front End=<0 0 500>)
    )) {module=AddVelocity001}
   &SampleSkeletalMeshSkeleton(Mesh:NiagaraDataInterfaceSkeletalMesh=$User.SkeletalMesh, Specific Bone Index=HlslExpression() {hlslExpression="(Particles.SpawnGroup*3)+2"})
   &SkeletalMeshLocation(
    Bone / Socket Index=HlslExpression() {hlslExpression="(Particles.SpawnGroup*3)+2"}
    Bone Sampling Mode=::NewEnumerator4
    Inherit Velocity=false
    "Inherit Velocity (Scale)"=0
    Interpolate Mesh Sampling=false
    MaskSpawnGroup=false
    Sampled Position Offset=<0 5 -10>
    Sampled Position Offset Coordinate Space:ENiagaraMeshTransforms=::NewEnumerator2
    Skeletal Mesh=$User.SkeletalMesh
   ) {enabled=false}
  ]
  particleUpdateScript [
   &AccelerationForce(
    Acceleration=&VectorFromCurve(VectorCurve=DataInterface() {propertyValues {
       XCurve {keys [<at 0.0017 132.1921>,<at 0.9905 -65.3336>]}
       YCurve {keys [<at -0.0011 1.0067>,<at 0.9988 0>]}
       ZCurve {keys [<at -0.0011 -300>,<at 0.9957 -200.05>]}
      }})
    Coordinate Space=::Local
   ) {enabled=false}
   &Drag(Drag=0.9, Use Linear Drag=true, Use Rotational Drag=false) {enabled=false}
   &InheritVelocity(
    Module.Limit Speed=true
    Velocity Limit=10000
    Velocity Scale=&MakeCustomVectorFromBool(Bool=$Particles.Initial.Locked, True Vector=<0 0 0>, False Vector=<-0.03 -0.03 -0.01>)
    Velocity Source=&TransformVector(Vector=$Engine.Owner.Velocity, Source Space=::World, Destination Space=::Local)
   ) {enabled=false}
   &FN_DivingAttributes() {enabled=false}
   &SolveForcesAndVelocity(Clamp Velocity=false, Limit Acceleration=false)
   &FN_UpdateRibbon(
    Enable Skeletal Position=true
    Initial Color:LinearColor=$Particles.Initial.Color
    Initial RibbonWidth:NiagaraFloat=$Particles.Initial.RibbonWidth
   )
   &ScaleColor(
    Scale Alpha=&FloatFromCurve(FloatCurve=DataInterface() {propertyValues {Curve {keys [
         {arriveTangent=1.8999, interpMode=RCIM_Cubic, leaveTangent=1.8999, tangentMode=RCTM_User, time=-0.0044, value=0.9894}
         {arriveTangent=-1.934, interpMode=RCIM_Cubic, leaveTangent=-1.934, tangentMode=RCTM_User, time=0.2157, value=1}
         {arriveTangent=-0.0191, interpMode=RCIM_Cubic, leaveTangent=-0.0191, tangentMode=RCTM_User, time=1.0011, value=0.002}
        ]}}})
    ScaleA=true
    ScaleRGB=false
   )
   &Material(Base Texture Time:NiagaraFloat=&Multiply_Float(A=$Emitter.Age, B=-1), PostMult=0.5455, Power=0.4408)
   &ScaleRibbonWidth(Ribbon Width Scale=&FloatFromCurve(FloatCurve=DataInterface() {propertyValues {Curve {keys [
         {arriveTangent=1.9737, interpMode=RCIM_Cubic, leaveTangent=1.9737, tangentMode=RCTM_User, time=0.0058, value=0.9363}
         {arriveTangent=-0.0131, interpMode=RCIM_Cubic, leaveTangent=-0.0131, time=0.4538, value=1.5093}
         <at 0.6635 0.9277>
        ]}}})) {enabled=false}
   &Color(
    Color=<rgba 1 1 1 0.8>
    Scale Alpha=&FloatFromCurve(
     FloatCurve=DataInterface() {propertyValues {Curve {keys [
         {arriveTangent=2.7939, interpMode=RCIM_Cubic, leaveTangent=2.7939, tangentMode=RCTM_User, time=-0.0046, value=1.0113}
         {arriveTangent=-1.0036, interpMode=RCIM_Cubic, leaveTangent=-4.3444, tangentMode=RCTM_Break, time=0.1387, value=1.232}
         <at 0.6 0.002>
        ]}}}
     Scale Curve=0.8
    )
   ) {enabled=false}
  ]
  rendererClasses [&NiagaraRibbonRendererProperties]
  renderers [{rendererClass=&NiagaraRibbonRendererProperties}]
 }
 FN_TEST_AXIS {
  bEnabled=false
  emitterUpdateScript [
   &EmitterState()
   &SpawnBurst_Instantaneous(Spawn Count=6, Spawn Probability=HlslExpression() {hlslExpression="Emitter.LoopCount == 0"}, Use Spawn Probability=true)
  ]
  particleSpawnScript [
   &FN_SetSpawnGroup_v2(Number of Spawn Groups=4)
   &FN_SetSpawnGroup()
   &InitializeParticle(
    Module.Write Color=true
    Module.Write Lifetime=true
    Module.Write Mass=true
    Module.Write Material Random=true
    Module.Write Position=false
    Module.Write Scale=false
    Module.Write SpriteRotation=false
    Module.Write SpriteSize=false
    Position=&SimulationPosition()
   )
  ]
  particleUpdateScript [
   &ParticleState(Kill Particles When Lifetime Has Elapsed=false)
   &SkeletalMeshLocation(
    Apply Sampled Orientation to Mesh Particles=true
    Bone / Socket Index=$Particles.SpawnGroup
    Bone Sampling Mode=::NewEnumerator4
    Inherit Velocity=false
    Interpolate Mesh Sampling=false
    MaskSpawnGroup=false
    Skeletal Mesh=$User.SkeletalMesh
   )
  ]
  rendererClasses [&NiagaraMeshRendererProperties]
  renderers [{rendererClass=&NiagaraMeshRendererProperties, rendererIndex=0}]
 }
 FireSmoke_Upward {
  emitterUpdateScript [
   &EmitterState(Life Cycle Mode=::NewEnumerator1, Loop Duration=1, UseLoopDelay=false)
   &SpawnRate(
    SpawnRate=&MakeCustomFloatFromBool(Bool=$User.bIsFrontEndPreview, True Float=80, False Float=0)
    Use Spawn Probability=true
   )
   &SpawnPerUnit(Max Movement Threshold=2000, Module.Use Max Movement Threshold=true, Module.Use Movement Tolerance=true, Use Spawn Probability=false)
   &SpawnPerUnit(
    Max Movement Threshold=6110
    Module.Use Max Movement Threshold=true
    Module.Use Movement Tolerance=true
    Movement Tolerance=105
    Spawn Spacing=70
    Use Spawn Probability=false
   ) {module=SpawnPerUnit001}
  ]
  particleSpawnScript [
   &FN_DivingAttributes_v2(SkeletalMesh:NiagaraDataInterfaceSkeletalMesh=$User.SkeletalMesh)
   &FN_SetSpawnGroup_v2(Number of Spawn Groups=4)
   /SetVariables(Particles.DiveCLAMP:NiagaraFloat=&ClampFloat(
     Float=&FN_FrontEndFloatModifier(Gameplay=$Particles.Dive, Front End=&Multiply_Float(A=$Particles.Dive, B=10))
     Min=-1
    )) {enabled=false, isSetParameters=true, module=SetVariables_#1}
   &InitializeParticle(
    Color=<rgba 1 1 1 0.5>
    Lifetime=&MakeCustomFloatFromBool(Bool=$User.bIsFrontEndPreview, True Float=0.5, False Float=&Lerp_Float(A=0.3, B=0.3, Alpha=$System.Dive))
    Module.Write Color=true
    Module.Write Lifetime=true
    Module.Write Mass=false
    Module.Write Material Random=true
    Module.Write Position=false
    Module.Write Scale=false
    Module.Write SpriteRotation=true
    Module.Write SpriteSize=true
    Position=&SimulationPosition()
    Sprite Rotation=&UniformRangedFloat(Minimum=0, Maximum=300)
    Sprite Size=&Vector2DFromFloat(Value=&RandomRangeFloat(Minimum=10, Maximum=20))
   )
   &DynamicMaterialParameters(
    Index 0 Param 2=&UniformRangedFloat(Minimum=0, Maximum=1)
    Index 0 Param 3=1
    Index 0 Param 4=&UniformRangedFloat(Minimum=1, Maximum=0.5)
    Module.Param0WriteEnabled=false
    Module.Param1WriteEnabled=true
    Module.Param2WriteEnabled=false
    Module.Param3WriteEnabled=true
   ) {enabled=false, module=DynamicMaterialParameters002}
   &AddVelocity(Velocity=&Lerp_Vector(
     A=&UniformRangedVector(Minimum=<-25 0 0>, Maximum=<25 0 0>)
     B=&UniformRangedVector(Minimum=<-50 0 0>, Maximum=<50 0 0>)
     Alpha=$Particles.DiveCLAMP
    )) {enabled=false}
   &AddVelocity(Velocity=&Lerp_Vector(A=<0 0 100>, B=<0 0 50>, Alpha=$System.Dive)) {module=AddVelocity001}
   &SkeletalMeshLocation(
    Bone / Socket Index=HlslExpression() {hlslExpression=((Particles.SpawnGroup*3)+2)}
    Bone Sampling Mode=::NewEnumerator4
    Inherit Velocity=false
    "Inherit Velocity (Scale)"=0
    Interpolate Mesh Sampling=false
    MaskSpawnGroup=false
    Skeletal Mesh=$User.SkeletalMesh
   )
   &CameraOffset(Camera Offset Amount=5)
   &SphereLocation(
    MaskSpawnGroup=false
    Override Seed=false
    Sphere Radius=5
    Surface Only=false
    UseNonUniformScale=false
    UseOffset=false
    UseSphereOrientation=false
   )
  ]
  particleUpdateScript [
   /SetVariables(Particles.DiveCLAMP:NiagaraFloat=&ClampFloat(Float=&FN_FrontEndFloatModifier(Gameplay=$Particles.Dive, Front End=&Multiply_Float(A=$Particles.Dive, B=10)))) {enabled=false, isSetParameters=true, module=SetVariables_#3}
   &ParticleState()
   &Drag(Use Linear Drag=true, Use Rotational Drag=false)
   &SolveForcesAndVelocity(Clamp Velocity=false, Limit Acceleration=false)
   &ScaleSpriteSize(Scale Factor=&Vector2DFromFloat(Value=&FloatFromCurve(
      FloatCurve=DataInterface() {propertyValues {Curve {keys [
          {arriveTangent=-0.0274, interpMode=RCIM_Cubic, leaveTangent=-0.0274, tangentMode=RCTM_User, time=0, value=1}
          {arriveTangent=0.3453, interpMode=RCIM_Cubic, leaveTangent=0.3453, tangentMode=RCTM_User, time=1, value=2}
         ]}}}
      Scale Curve=&Lerp_Float(A=1, B=1.5, Alpha=$System.Dive)
     )))
   &ScaleColor(
    Scale Alpha=&FloatFromCurve(FloatCurve=DataInterface() {propertyValues {Curve {keys [
         {arriveTangent=8.7245, interpMode=RCIM_Cubic, leaveTangent=8.7245, tangentMode=RCTM_User, time=0, value=0}
         {arriveTangent=0.4876, interpMode=RCIM_Cubic, leaveTangent=0.4876, time=0.2, value=1}
         {arriveTangent=-1.25, interpMode=RCIM_Cubic, leaveTangent=-1.25, time=0.6306, value=0.3075}
         {arriveTangent=-0.5281, interpMode=RCIM_Cubic, leaveTangent=-0.5281, tangentMode=RCTM_User, time=1, value=0}
        ]}}})
    Scale RGB=&VectorFromFloat(Value=&FloatFromCurve(FloatCurve=DataInterface() {propertyValues {Curve {keys [
          {arriveTangent=-2.3743, interpMode=RCIM_Cubic, leaveTangent=-2.3743, tangentMode=RCTM_User, time=0, value=1}
          <at 0.1 0.7>
          <at 1 0.5 RCIM_Cubic>
         ]}}}))
    ScaleA=true
    ScaleRGB=true
   ) {enabled=false}
   &ScaleColor(
    Scale Alpha=&FloatFromCurve(
     FloatCurve=DataInterface() {propertyValues {Curve {keys [
         {interpMode=RCIM_Cubic, tangentMode=RCTM_User, time=0, value=0}
         {arriveTangent=2.6667, interpMode=RCIM_Cubic, leaveTangent=2.6667, time=0.1, value=0.15}
         {arriveTangent=-2.8121, interpMode=RCIM_Cubic, leaveTangent=-2.8121, tangentMode=RCTM_User, time=0.3, value=0.8}
         {tangentMode=RCTM_User, time=1, value=0}
        ]}}}
     Scale Curve=&Lerp_Float(A=1, B=3, Alpha=$System.Dive)
    )
    Scale RGB=&MakeVectorFromLinearColorRGB(Color=&ColorFromCurve(ColorCurve=DataInterface() {propertyValues {
        AlphaCurve {keys [<at 0 1 RCIM_Cubic>,<at 1 1 RCIM_Cubic>]}
        BlueCurve {keys [<at 0.2621 0.02>,<at 0.5131 0.03>,<at 1 0.03 RCIM_Cubic>]}
        GreenCurve {keys [<at 0.2621 0.02>,<at 0.5131 0.2887>,<at 1 0.2887 RCIM_Cubic>]}
        RedCurve {keys [<at 0.2621 1>,<at 0.5131 1>,<at 1 1 RCIM_Cubic>]}
       }}))
    ScaleA=true
    ScaleRGB=true
   ) {module=ScaleColor001}
   &DynamicMaterialParameters(
    Index 0 Param 1=&Add_Float(
     A=&FloatFromCurve(
      FloatCurve=DataInterface() {propertyValues {Curve {keys [{interpMode=RCIM_Cubic, tangentMode=RCTM_User, time=0, value=1}, {interpMode=RCIM_Cubic, tangentMode=RCTM_User, time=1, value=0}]}}}
      Scale Curve=0.7
     )
     B=&Multiply_Float(A=$Particles.Initial.MaterialRandom, B=0.5)
    )
    Index 0 Param 3=&FloatFromCurve(
     FloatCurve=DataInterface() {propertyValues {Curve {keys [
         {arriveTangent=-0.0944, interpMode=RCIM_Cubic, time=0, value=0}
         {arriveTangent=2.5569, interpMode=RCIM_Cubic, leaveTangent=2.5569, time=0.0519, value=2.3053}
         <at 0.5008 1.2804>
        ]}}}
     Scale Curve=0.2
    )
    Index 0 Param 4=1
    Index 2 Param 1=&Lerp_Float(A=0.5, B=&Add_Float(A=$Particles.Initial.MaterialRandom, B=1.5), Alpha=$Particles.NormalizedAge)
    Index 2 Param 3=2
    Index 2 Param 4=2
    Module.Param0WriteEnabled=true
    Module.Param0WriteEnabled1=true
    Module.Param0WriteEnabled2=true
    Module.Param1WriteEnabled=true
    Module.Param1WriteEnabled1=true
    Module.Param1WriteEnabled2=true
    Module.Param2WriteEnabled=true
    Module.Param2WriteEnabled1=true
    Module.Param2WriteEnabled2=true
    Module.Param3WriteEnabled=true
    Module.Param3WriteEnabled1=true
    Module.Param3WriteEnabled2=true
    Write Parameter Index 1=true
    Write Parameter Index 2=true
   )
  ]
  rendererClasses [&NiagaraSpriteRendererProperties]
  renderers [{rendererClass=&NiagaraSpriteRendererProperties}]
 }
 Flames_FlowHands {
  bEnabled=false
  emitterUpdateScript [
   &EmitterState(Life Cycle Mode=::NewEnumerator1, Loop Duration=1, UseLoopDelay=false)
   &SpawnRate(SpawnRate=&Lerp_Float(A=20, B=10, Alpha=$System.Dive), Use Spawn Probability=false)
  ]
  particleSpawnScript [
   &InitializeParticle(
    Color=&MakeLinearColorFromFloat(Value:NiagaraFloat=&FloatFromCurve(
      FloatCurve=DataInterface() {propertyValues {Curve {keys [
          {interpMode=RCIM_Cubic, tangentMode=RCTM_User, time=0, value=1}
          {arriveTangent=-3.3333, interpMode=RCIM_Cubic, leaveTangent=-3.3333, tangentMode=RCTM_User, time=0.5111, value=2}
          <at 0.8745 0.4925>
          <at 2.0414 0.1937>
          {interpMode=RCIM_Cubic, tangentMode=RCTM_User, time=3, value=1}
         ]}}}
      CurveIndex=$Emitter.LoopedAge
     ))
    Lifetime=&UniformRangedFloat(Minimum=1.2, Maximum=1.8)
    Module.Write Color=true
    Module.Write Lifetime=true
    Module.Write Mass=true
    Module.Write Material Random=true
    Module.Write Position=false
    Module.Write Scale=false
    Module.Write SpriteRotation=true
    Module.Write SpriteSize=true
    Position=&SimulationPosition()
    Sprite Rotation=&UniformRangedFloat(Minimum=0, Maximum=350)
    Sprite Size=&Vector2DFromFloat(Value=&RandomRangeFloat(Minimum=2, Maximum=3))
   )
   /SetVariables(Particles.BoneFlowAngle:NiagaraFloat=&UniformRangedFloat(Minimum=0, Maximum=6.4)) {isSetParameters=true, module=SetVariables_#6}
   &FN_BoneFlow(
    Angle:NiagaraFloat=$Particles.BoneFlowAngle
    BoneID=HlslExpression() {hlslExpression="(Particles.UniqueID && 1) *3"}
    Radius:NiagaraFloat=&FloatFromCurve(FloatCurve=DataInterface() {propertyValues {Curve {keys [<at 0 1>,<at 1 0>]}}}, Scale Curve=10)
    SkeletalMesh:NiagaraDataInterfaceSkeletalMesh=$User.SkeletalMesh
    Write Position Attribute=true
   ) {enabled=false}
  ]
  particleUpdateScript [
   &ScaleColor(
    Scale RGB=&MakeVectorFromLinearColorRGB(Color=&ColorFromCurve(ColorCurve=DataInterface() {propertyValues {
        AlphaCurve {keys [<at 0 1 RCIM_Cubic>,<at 1 1 RCIM_Cubic>]}
        BlueCurve {keys [
          <at 0.1333 0 RCIM_Cubic>
          <at 0.5316 0.06>
          {arriveTangent=-0.2801, interpMode=RCIM_Cubic, leaveTangent=-0.2801, time=0.6008, value=0.05}
          <at 0.7457 0>
         ]}
        GreenCurve {keys [
          <at 0.1333 0.0167 RCIM_Cubic>
          <at 0.5316 0.3187>
          {arriveTangent=-1.4879, interpMode=RCIM_Cubic, leaveTangent=-1.4879, time=0.6008, value=0.1133}
          <at 0.7457 0>
         ]}
        RedCurve {keys [
          <at 0.1333 1 RCIM_Cubic>
          <at 0.5316 2>
          {arriveTangent=-4.669, interpMode=RCIM_Cubic, leaveTangent=-4.669, time=0.6008, value=1}
          <at 0.7457 1>
         ]}
       }}))
    ScaleA=true
    ScaleRGB=true
   ) {module=ScaleColor001}
   &FN_BoneFlow(
    Angle:NiagaraFloat=$Particles.BoneFlowAngle
    BoneID=HlslExpression() {hlslExpression="((Particles.UniqueID && 1) *1)+0"}
    Height:NiagaraFloat=&FloatFromCurve(FloatCurve=DataInterface() {propertyValues {Curve {keys [
         {arriveTangent=1.3633, interpMode=RCIM_Cubic, leaveTangent=1.3633, tangentMode=RCTM_User, time=0, value=0}
         {arriveTangent=0.3204, interpMode=RCIM_Cubic, leaveTangent=0.3204, tangentMode=RCTM_User, time=0.9897, value=1.1}
        ]}}})
    Interpolate Mesh Sampling=false
    Radius:NiagaraFloat=&FloatFromCurve(FloatCurve=DataInterface() {propertyValues {Curve {keys [
         {arriveTangent=0.8832, interpMode=RCIM_Cubic, time=0.1742, value=0.2546}
         <at 0.4784 0.2906>
         {arriveTangent=-0.2267, interpMode=RCIM_Cubic, leaveTangent=-0.2267, time=0.6089, value=0.5975}
         <at 0.7465 0.2298>
         {arriveTangent=-2.8131, interpMode=RCIM_Cubic, leaveTangent=-2.8131, tangentMode=RCTM_User, time=0.8737, value=0.065}
        ]}}})
    SkeletalMesh:NiagaraDataInterfaceSkeletalMesh=$User.SkeletalMesh
    Write Position Attribute=true
   ) {module=FN_BoneFlow002}
   /SetVariables(Particles.TestSplineBasedVector:Vector3f=$Output.FN_BoneFlow.SplineTangent) {isSetParameters=true, module=SetVariables_#7}
   &SkeletalMeshLocation(
    Apply Sampled Orientation to Mesh Particles=true
    Bone / Socket Index=$Particles.SpawnGroup
    Bone Sampling Mode=::NewEnumerator4
    Inherit Velocity=false
    Interpolate Mesh Sampling=false
    MaskSpawnGroup=false
    Skeletal Mesh=$User.SkeletalMesh
   ) {enabled=false}
   &AddVelocity(
    Scale Added Velocity=<100 100 100>
    Velocity=&RotateVectorByQuaternion(
     Vector To Rotate=<1 0 0>
     Quaternion=&FindQuatBetweenTwoVectors(
      From Vector=Linked() {linkedVariable {name=Particles.Position, type=NiagaraPosition}}
      To Vector=$Particles.SkeletalMeshLocation.SampledPosition
     )
    )
   ) {enabled=false}
   &ParticleState()
   &ScaleSpriteSize(Scale Factor=&Vector2DFromFloat(Value=&FloatFromCurve(
      FloatCurve=DataInterface() {propertyValues {Curve {keys [
          {arriveTangent=18.9237, interpMode=RCIM_Cubic, leaveTangent=18.9237, tangentMode=RCTM_User, time=0, value=1}
          <at 0.5509 3.8421>
          <at 0.7003 9.9984>
          {arriveTangent=-17.3587, interpMode=RCIM_Cubic, leaveTangent=-17.3586, tangentMode=RCTM_User, time=0.8294, value=5.9457}
          {arriveTangent=-7.3072, interpMode=RCIM_Cubic, leaveTangent=-7.3072, tangentMode=RCTM_User, time=0.9951, value=0.5}
         ]}}}
      Scale Curve=&Lerp_Float(A=1, B=2, Alpha=$System.Dive)
     )))
   &SolveForcesAndVelocity(Clamp Velocity=false, Limit Acceleration=false)
   /Niagara/Modules/Update/SubUV/SubUVAnimation(Number Of Frames=&UniformRangedFloat(Minimum=20, Maximum=80))
   &CameraOffset(Camera Offset Amount=&FloatFromCurve(
     FloatCurve=DataInterface() {propertyValues {Curve {keys [<at 0.2 0 RCIM_Constant>,<at 0.3 0.5 RCIM_Constant>,<at 0.6 -0.645 RCIM_Constant>,<at 0.8548 0.3856 RCIM_Constant>]}}}
     Scale Curve=20
    ))
  ]
  rendererClasses [&NiagaraSpriteRendererProperties]
  renderers [{rendererClass=&NiagaraSpriteRendererProperties}]
 }
 Flames_FlowHands002 {
  emitterUpdateScript [
   &EmitterState(Life Cycle Mode=::NewEnumerator1, Loop Duration=1, UseLoopDelay=false)
   &SpawnRate(
    SpawnRate=&MakeCustomFloatFromBool(Bool=$User.bIsFrontEndPreview, True Float=30, False Float=&Lerp_Float(A=20, B=5, Alpha=$System.Dive))
    Use Spawn Probability=false
   )
  ]
  particleSpawnScript [
   &InitializeParticle(
    Lifetime=&RandomRangeFloat(Minimum=1.2, Maximum=1.8)
    Module.Write Color=true
    Module.Write Lifetime=true
    Module.Write Mass=true
    Module.Write Material Random=true
    Module.Write Position=false
    Module.Write Scale=false
    Module.Write SpriteRotation=true
    Module.Write SpriteSize=true
    Position=&SimulationPosition()
    Sprite Rotation=&UniformRangedFloat(Minimum=0, Maximum=350)
    Sprite Size=&Vector2DFromFloat(Value=&RandomRangeFloat(Minimum=3, Maximum=4))
   )
   /SetVariables(Particles.BoneFlowAngle:NiagaraFloat=&UniformRangedFloat(Minimum=0, Maximum=6.4)) {isSetParameters=true, module=SetVariables_#6}
   &FN_BoneFlow(
    Angle:NiagaraFloat=$Particles.BoneFlowAngle
    BoneID=HlslExpression() {hlslExpression="(Particles.UniqueID && 1) *3"}
    Radius:NiagaraFloat=&FloatFromCurve(FloatCurve=DataInterface() {propertyValues {Curve {keys [<at 0 1>,<at 1 0>]}}}, Scale Curve=10)
    SkeletalMesh:NiagaraDataInterfaceSkeletalMesh=$User.SkeletalMesh
    Write Position Attribute=true
   ) {enabled=false}
  ]
  particleUpdateScript [
   &ScaleColor(
    Scale RGB=&MakeVectorFromLinearColorRGB(Color=&MultiplyLinearColor(
      Linear Color=&ColorFromCurve(ColorCurve=DataInterface() {propertyValues {
         AlphaCurve {keys [<at 0 1 RCIM_Cubic>,<at 1 1 RCIM_Cubic>]}
         BlueCurve {keys [
           <at 0.0095 0.3>
           <at 0.3138 0>
           <at 0.4982 0.02>
           {arriveTangent=0.1542, interpMode=RCIM_Cubic, leaveTangent=0.1542, time=0.7321, value=0.01}
           {interpMode=RCIM_Cubic, leaveTangent=-0.1401, time=0.9622, value=0.1}
          ]}
         GreenCurve {keys [
           <at 0.0095 0.7083>
           <at 0.3138 0.0082>
           <at 0.4982 0.4283>
           {arriveTangent=0.1415, interpMode=RCIM_Cubic, leaveTangent=0.1415, time=0.7321, value=0.2575}
           {interpMode=RCIM_Cubic, leaveTangent=-0.7439, time=0.9622, value=0.1}
          ]}
         RedCurve {keys [
           <at 0.0095 1>
           <at 0.3138 0.599>
           <at 0.4982 1>
           {arriveTangent=0.6185, interpMode=RCIM_Cubic, leaveTangent=0.6185, time=0.7321, value=1}
           <at 0.9622 1 RCIM_Cubic>
          ]}
        }})
      Scale Factor=&MakeVector4_FromFloat(Float=&FloatFromCurve(
        FloatCurve=DataInterface() {propertyValues {Curve {keys [{interpMode=RCIM_Cubic, tangentMode=RCTM_User, time=0, value=0.9}, {interpMode=RCIM_Cubic, tangentMode=RCTM_User, time=1, value=1}]}}}
        Scale Curve=200
       ))
     ))
    ScaleA=true
    ScaleRGB=true
   ) {module=ScaleColor001}
   &FN_BoneFlow(
    Angle:NiagaraFloat=$Particles.BoneFlowAngle
    BoneID=HlslExpression() {hlslExpression="((Particles.UniqueID && 1) *1)+0"}
    Height:NiagaraFloat=&FloatFromCurve(FloatCurve=DataInterface() {propertyValues {Curve {keys [
         {arriveTangent=1.3633, interpMode=RCIM_Cubic, leaveTangent=1.3633, tangentMode=RCTM_User, time=0, value=0}
         {arriveTangent=0.3204, interpMode=RCIM_Cubic, leaveTangent=0.3204, tangentMode=RCTM_User, time=0.9897, value=1.1}
        ]}}})
    Interpolate Mesh Sampling=false
    Radius:NiagaraFloat=&FloatFromCurve(FloatCurve=DataInterface() {propertyValues {Curve {keys [
         {arriveTangent=0.8832, interpMode=RCIM_Cubic, time=0.1742, value=0.2546}
         <at 0.4784 0.2906>
         {arriveTangent=-0.2267, interpMode=RCIM_Cubic, leaveTangent=-0.2267, time=0.6089, value=0.5975}
         <at 0.7465 0.2298>
         {arriveTangent=-2.8131, interpMode=RCIM_Cubic, leaveTangent=-2.8131, tangentMode=RCTM_User, time=0.8737, value=0.065}
        ]}}})
    SkeletalMesh:NiagaraDataInterfaceSkeletalMesh=$User.SkeletalMesh
    Write Position Attribute=true
   ) {module=FN_BoneFlow002}
   /SetVariables(Particles.TestSplineBasedVector:Vector3f=$Output.FN_BoneFlow.SplineTangent) {isSetParameters=true, module=SetVariables_#7}
   &DynamicMaterialParameters(
    Index 0 Param 1=&FloatFromCurve(FloatCurve=DataInterface() {propertyValues {Curve {keys [
         {arriveTangent=2.125, interpMode=RCIM_Cubic, time=0, value=1}
         {arriveTangent=3.7313, interpMode=RCIM_Cubic, leaveTangent=3.7313, tangentMode=RCTM_User, time=0.1, value=0.15}
         <at 0.4 1 RCIM_Cubic>
        ]}}})
    Index 0 Param 2=1
    Index 0 Param 3=&FloatFromCurve(
     FloatCurve=DataInterface() {propertyValues {Curve {keys [
         {arriveTangent=-0.3008, interpMode=RCIM_Cubic, leaveTangent=-0.3008, tangentMode=RCTM_User, time=0, value=0}
         {interpMode=RCIM_Cubic, tangentMode=RCTM_User, time=0.5809, value=1}
        ]}}}
     Scale Curve=8
    )
    Index 0 Param 4=1
    Module.Param0WriteEnabled=true
    Module.Param0WriteEnabled1=true
    Module.Param1WriteEnabled=true
    Module.Param1WriteEnabled1=true
    Module.Param2WriteEnabled=true
    Module.Param2WriteEnabled1=true
    Module.Param3WriteEnabled=true
    Module.Param3WriteEnabled1=true
    Write Parameter Index 1=true
   ) {module=DynamicMaterialParameters001}
   &SkeletalMeshLocation(
    Apply Sampled Orientation to Mesh Particles=true
    Bone / Socket Index=$Particles.SpawnGroup
    Bone Sampling Mode=::NewEnumerator4
    Inherit Velocity=false
    Interpolate Mesh Sampling=false
    MaskSpawnGroup=false
    Skeletal Mesh=$User.SkeletalMesh
   ) {enabled=false}
   &AddVelocity(
    Scale Added Velocity=<100 100 100>
    Velocity=&RotateVectorByQuaternion(
     Vector To Rotate=<1 0 0>
     Quaternion=&FindQuatBetweenTwoVectors(
      From Vector=Linked() {linkedVariable {name=Particles.Position, type=NiagaraPosition}}
      To Vector=$Particles.SkeletalMeshLocation.SampledPosition
     )
    )
   ) {enabled=false}
   &ParticleState()
   &ScaleSpriteSize(Scale Factor=&Vector2DFromFloat(Value=&FloatFromCurve(
      FloatCurve=DataInterface() {propertyValues {Curve {keys [
          {arriveTangent=18.9237, interpMode=RCIM_Cubic, leaveTangent=18.9237, tangentMode=RCTM_User, time=0, value=1}
          <at 0.5509 3.8421>
          <at 0.7003 9.9984>
          {arriveTangent=-17.3587, interpMode=RCIM_Cubic, leaveTangent=-17.3586, tangentMode=RCTM_User, time=0.8294, value=5.9457}
          {arriveTangent=-7.3072, interpMode=RCIM_Cubic, leaveTangent=-7.3072, tangentMode=RCTM_User, time=0.9951, value=0.5}
         ]}}}
      Scale Curve=&Lerp_Float(A=1, B=1.5, Alpha=$System.Dive)
     )))
   &SolveForcesAndVelocity(Clamp Velocity=false, Limit Acceleration=false)
   &SubUVAnimation(Number Of Frames=&UniformRangedFloat(Minimum=20, Maximum=20)) {enabled=false}
   &CameraOffset(Camera Offset Amount=&FloatFromCurve(
     FloatCurve=DataInterface() {propertyValues {Curve {keys [<at 0.2 0 RCIM_Constant>,<at 0.3 0.5 RCIM_Constant>,<at 0.6 -0.645 RCIM_Constant>,<at 0.8548 0.3856 RCIM_Constant>]}}}
     Scale Curve=20
    ))
   /Niagara/Modules/Update/SubUV/V2/SubUVAnimation(
    Sub UV Play Rate:NiagaraFloat=&Lerp_Float(A=1, B=3, Alpha=$System.Dive)
    SubUV Animation Mode:ENiagara_SubUVLookupMode=::NewEnumerator3
   ) {module=SubUVAnimation001}
  ]
  rendererClasses [&NiagaraSpriteRendererProperties]
  renderers [{rendererClass=&NiagaraSpriteRendererProperties}]
 }
 Flames_FlowHandsDark {
  bEnabled=false
  emitterUpdateScript [
   &EmitterState(Life Cycle Mode=::NewEnumerator1, Loop Duration=1, UseLoopDelay=false)
   &SpawnRate(
    Spawn Probability=&FloatFromCurve(
     FloatCurve=DataInterface() {propertyValues {Curve {keys [<at 0 1>,<at 0.2898 0.9385>,<at 0.3858 0>,<at 3 0>]}}}
     CurveIndex=&Multiply_Float(A=$Emitter.NormalizedLoopAge, B=&RandomRangeFloat(Minimum=1, Maximum=5))
    )
    SpawnRate=50
    Use Spawn Probability=true
   )
  ]
  particleSpawnScript [
   &InitializeParticle(
    Lifetime=&UniformRangedFloat(Minimum=1.5, Maximum=2)
    Module.Write Color=true
    Module.Write Lifetime=true
    Module.Write Mass=true
    Module.Write Material Random=true
    Module.Write Position=false
    Module.Write Scale=false
    Module.Write SpriteRotation=true
    Module.Write SpriteSize=true
    Position=&SimulationPosition()
    Sprite Rotation=&UniformRangedFloat(Minimum=0, Maximum=350)
    Sprite Size=&Vector2DFromFloat(Value=2)
   )
   /SetVariables(Particles.BoneFlowAngle:NiagaraFloat=&UniformRangedFloat(Minimum=0, Maximum=6.4)) {isSetParameters=true, module=SetVariables_#6}
   &FN_BoneFlow(
    Angle:NiagaraFloat=$Particles.BoneFlowAngle
    BoneID=HlslExpression() {hlslExpression="(Particles.UniqueID && 1) *3"}
    Radius:NiagaraFloat=&FloatFromCurve(FloatCurve=DataInterface() {propertyValues {Curve {keys [<at 0 1>,<at 1 0>]}}}, Scale Curve=10)
    SkeletalMesh:NiagaraDataInterfaceSkeletalMesh=$User.SkeletalMesh
    Write Position Attribute=true
   ) {enabled=false}
  ]
  particleUpdateScript [
   &ScaleColor(
    Scale RGB=&MakeVectorFromLinearColorRGB(Color=&ColorFromCurve(ColorCurve=DataInterface() {propertyValues {
        AlphaCurve {keys [<at 0 1 RCIM_Cubic>,<at 1 1 RCIM_Cubic>]}
        BlueCurve {keys [<at 0.2502 0 RCIM_Cubic>,<at 0.292 0.3>,<at 0.4262 0.0115>,<at 0.5388 0.0141>,<at 0.6008 0.15 RCIM_Cubic>]}
        GreenCurve {keys [<at 0.2502 0.0161 RCIM_Cubic>,<at 0.292 0.57>,<at 0.4262 0.026>,<at 0.5388 0.0267>,<at 0.6008 0.34 RCIM_Cubic>]}
        RedCurve {keys [<at 0.2502 0.1615 RCIM_Cubic>,<at 0.292 3>,<at 0.4262 0.2292>,<at 0.5388 0.1406>,<at 0.6008 3 RCIM_Cubic>]}
       }}))
    ScaleA=true
    ScaleRGB=true
   ) {enabled=false}
   &ScaleColor(
    Scale RGB=&MakeVectorFromLinearColorRGB(Color=&ColorFromCurve(ColorCurve=DataInterface() {propertyValues {
        AlphaCurve {keys [<at 0 1 RCIM_Cubic>,<at 1 1 RCIM_Cubic>]}
        BlueCurve {keys [
          <at 0.1333 0 RCIM_Cubic>
          {arriveTangent=0.0319, interpMode=RCIM_Cubic, leaveTangent=0.0319, time=0.2777, value=0.05}
          <at 0.4627 0.03>
          <at 0.6031 0.015>
         ]}
        GreenCurve {keys [
          <at 0.1333 0 RCIM_Cubic>
          {arriveTangent=0.0663, interpMode=RCIM_Cubic, leaveTangent=0.0663, time=0.2777, value=0.0817}
          <at 0.4627 0.3533>
          <at 0.6031 0.0312>
         ]}
        RedCurve {keys [
          <at 0.1333 0.01 RCIM_Cubic>
          {arriveTangent=1.043, interpMode=RCIM_Cubic, leaveTangent=1.043, time=0.2777, value=1}
          <at 0.4627 1>
          <at 0.6031 0.5>
         ]}
       }}))
    ScaleA=true
    ScaleRGB=true
   ) {module=ScaleColor001}
   &FN_BoneFlow(
    Angle:NiagaraFloat=$Particles.BoneFlowAngle
    BoneID=HlslExpression() {hlslExpression="((Particles.UniqueID && 1) *1)+0"}
    Height:NiagaraFloat=&FloatFromCurve(FloatCurve=DataInterface() {propertyValues {Curve {keys [
         {arriveTangent=3.7027, interpMode=RCIM_Cubic, leaveTangent=3.7027, tangentMode=RCTM_User, time=0, value=0}
         {arriveTangent=0.3204, interpMode=RCIM_Cubic, leaveTangent=0.3204, tangentMode=RCTM_User, time=0.9897, value=1.2}
        ]}}})
    Radius:NiagaraFloat=&FloatFromCurve(
     FloatCurve=DataInterface() {propertyValues {Curve {keys [
         <at 0.23 0.3473 RCIM_Cubic>
         {arriveTangent=1.0971, interpMode=RCIM_Cubic, leaveTangent=1.0971, time=0.6661, value=0.6553}
         {arriveTangent=-2.8131, interpMode=RCIM_Cubic, leaveTangent=-2.8131, tangentMode=RCTM_User, time=0.863, value=1.0417}
        ]}}}
     Scale Curve=5
    )
    SkeletalMesh:NiagaraDataInterfaceSkeletalMesh=$User.SkeletalMesh
    Write Position Attribute=true
   ) {enabled=false, module=FN_BoneFlow001}
   &FN_BoneFlow(
    Angle:NiagaraFloat=$Particles.BoneFlowAngle
    BoneID=HlslExpression() {hlslExpression="((Particles.UniqueID && 1) *1)+0"}
    Height:NiagaraFloat=&FloatFromCurve(FloatCurve=DataInterface() {propertyValues {Curve {keys [
         {arriveTangent=2.8308, interpMode=RCIM_Cubic, leaveTangent=2.8308, tangentMode=RCTM_User, time=0, value=0}
         {arriveTangent=0.3204, interpMode=RCIM_Cubic, leaveTangent=0.3204, tangentMode=RCTM_User, time=0.9897, value=1.1}
        ]}}})
    Interpolate Mesh Sampling=false
    Radius:NiagaraFloat=&FloatFromCurve(FloatCurve=DataInterface() {propertyValues {Curve {keys [
         {arriveTangent=0.8832, interpMode=RCIM_Cubic, time=0.1742, value=0.2546}
         <at 0.4613 0.4858>
         {arriveTangent=1.3839, interpMode=RCIM_Cubic, leaveTangent=1.3839, time=0.6661, value=0.6553}
         {arriveTangent=-2.8131, interpMode=RCIM_Cubic, leaveTangent=-2.8131, tangentMode=RCTM_User, time=0.863, value=1.0417}
        ]}}})
    SkeletalMesh:NiagaraDataInterfaceSkeletalMesh=$User.SkeletalMesh
    Write Position Attribute=true
   ) {module=FN_BoneFlow002}
   /SetVariables(Particles.TestSplineBasedVector:Vector3f=$Output.FN_BoneFlow.SplineTangent) {isSetParameters=true, module=SetVariables_#7}
   &SkeletalMeshLocation(
    Apply Sampled Orientation to Mesh Particles=true
    Bone / Socket Index=$Particles.SpawnGroup
    Bone Sampling Mode=::NewEnumerator4
    Inherit Velocity=false
    Interpolate Mesh Sampling=false
    MaskSpawnGroup=false
    Skeletal Mesh=$User.SkeletalMesh
   ) {enabled=false}
   &AddVelocity(
    Scale Added Velocity=<100 100 100>
    Velocity=&RotateVectorByQuaternion(
     Vector To Rotate=<1 0 0>
     Quaternion=&FindQuatBetweenTwoVectors(
      From Vector=Linked() {linkedVariable {name=Particles.Position, type=NiagaraPosition}}
      To Vector=$Particles.SkeletalMeshLocation.SampledPosition
     )
    )
   ) {enabled=false}
   &ParticleState()
   &ScaleSpriteSize(Scale Factor=&Vector2DFromFloat(Value=&FloatFromCurve(
      FloatCurve=DataInterface() {propertyValues {Curve {keys [
          {arriveTangent=3.5027, interpMode=RCIM_Cubic, leaveTangent=3.5027, tangentMode=RCTM_User, time=0, value=1}
          {arriveTangent=-0.5025, interpMode=RCIM_Cubic, leaveTangent=-0.5025, time=0.7135, value=3.5838}
          {arriveTangent=-7.3072, interpMode=RCIM_Cubic, leaveTangent=-7.3072, tangentMode=RCTM_User, time=0.9951, value=0.5}
         ]}}}
      Scale Curve=3
     )))
   &SolveForcesAndVelocity(Clamp Velocity=false, Limit Acceleration=false)
   /Niagara/Modules/Update/SubUV/SubUVAnimation(Number Of Frames=100)
   &CameraOffset(Camera Offset Amount=&FloatFromCurve(FloatCurve=DataInterface() {propertyValues {Curve {keys [<at 0.4219 0>,<at 0.5791 0.5>,<at 0.6831 -0.5004>]}}}, Scale Curve=20))
  ]
  rendererClasses [&NiagaraSpriteRendererProperties]
  renderers [
   {rendererClass=&NiagaraSpriteRendererProperties}
   {rendererClass=&NiagaraSpriteRendererProperties, rendererIndex=1}
  ]
 }
 Flames_FlowLegs {
  bEnabled=false
  emitterUpdateScript [
   &EmitterState(Life Cycle Mode=::NewEnumerator1, Loop Duration=1, UseLoopDelay=false)
   &SpawnRate(SpawnRate=40, Use Spawn Probability=false)
  ]
  particleSpawnScript [
   &InitializeParticle(
    Lifetime=&UniformRangedFloat(Minimum=0.8, Maximum=2)
    Module.Write Color=true
    Module.Write Lifetime=true
    Module.Write Mass=true
    Module.Write Material Random=true
    Module.Write Position=false
    Module.Write Scale=false
    Module.Write SpriteRotation=true
    Module.Write SpriteSize=true
    Position=&SimulationPosition()
    Sprite Rotation=&Lerp_Float(A=0, B=&RandomRangeFloat(Minimum=-90, Maximum=90), Alpha=$System.Dive)
    Sprite Size=&Vector2DFromFloat(Value=&RandomRangeFloat(Minimum=4, Maximum=6))
   )
   /SetVariables(Particles.BoneFlowAngle:NiagaraFloat=&UniformRangedFloat(Minimum=0, Maximum=6.4)) {isSetParameters=true, module=SetVariables_#6}
   &FN_BoneFlow(
    Angle:NiagaraFloat=$Particles.BoneFlowAngle
    BoneID=HlslExpression() {hlslExpression="(Particles.UniqueID && 1) *3"}
    Radius:NiagaraFloat=&FloatFromCurve(FloatCurve=DataInterface() {propertyValues {Curve {keys [<at 0 1>,<at 1 0>]}}}, Scale Curve=10)
    SkeletalMesh:NiagaraDataInterfaceSkeletalMesh=$User.SkeletalMesh
    Write Position Attribute=true
   ) {enabled=false}
  ]
  particleUpdateScript [
   &ScaleColor(
    Scale RGB=&MakeVectorFromLinearColorRGB(Color=&ColorFromCurve(ColorCurve=DataInterface() {propertyValues {
        AlphaCurve {keys [<at 0 1 RCIM_Cubic>,<at 1 1 RCIM_Cubic>]}
        BlueCurve {keys [<at 0.1333 0 RCIM_Cubic>,<at 0.5316 0.05>,<at 0.6008 0.015 RCIM_Cubic>]}
        GreenCurve {keys [<at 0.1333 0 RCIM_Cubic>,<at 0.5316 0.2083>,<at 0.6008 0.034 RCIM_Cubic>]}
        RedCurve {keys [<at 0.1333 1 RCIM_Cubic>,<at 0.5316 1>,<at 0.6008 0.3 RCIM_Cubic>]}
       }}))
    ScaleA=true
    ScaleRGB=true
   )
   &FN_BoneFlow(
    Angle:NiagaraFloat=$Particles.BoneFlowAngle
    BoneID=HlslExpression() {hlslExpression="((Particles.UniqueID && 1) *1)+2"}
    Height:NiagaraFloat=&FloatFromCurve(FloatCurve=DataInterface() {propertyValues {Curve {keys [
         {arriveTangent=3.7027, interpMode=RCIM_Cubic, leaveTangent=3.7027, tangentMode=RCTM_User, time=0, value=0}
         {arriveTangent=0.3204, interpMode=RCIM_Cubic, leaveTangent=0.3204, tangentMode=RCTM_User, time=0.9897, value=1.1}
        ]}}})
    Interpolate Mesh Sampling=false
    Radius:NiagaraFloat=&FloatFromCurve(
     FloatCurve=DataInterface() {propertyValues {Curve {keys [
         <at 0.23 0.3473 RCIM_Cubic>
         {arriveTangent=1.0971, interpMode=RCIM_Cubic, leaveTangent=1.0971, time=0.6661, value=0.6553}
         {arriveTangent=-2.8131, interpMode=RCIM_Cubic, leaveTangent=-2.8131, tangentMode=RCTM_User, time=0.863, value=1.0417}
        ]}}}
     Scale Curve=5
    )
    SkeletalMesh:NiagaraDataInterfaceSkeletalMesh=$User.SkeletalMesh
    Write Position Attribute=true
   ) {module=FN_BoneFlow001}
   /SetVariables(Particles.TestSplineBasedVector:Vector3f=$Output.FN_BoneFlow.SplineTangent) {isSetParameters=true, module=SetVariables_#7}
   &SkeletalMeshLocation(
    Apply Sampled Orientation to Mesh Particles=true
    Bone / Socket Index=$Particles.SpawnGroup
    Bone Sampling Mode=::NewEnumerator4
    Inherit Velocity=false
    Interpolate Mesh Sampling=false
    MaskSpawnGroup=false
    Skeletal Mesh=$User.SkeletalMesh
   ) {enabled=false}
   &AddVelocity(
    Scale Added Velocity=<100 100 100>
    Velocity=&RotateVectorByQuaternion(
     Vector To Rotate=<1 0 0>
     Quaternion=&FindQuatBetweenTwoVectors(
      From Vector=Linked() {linkedVariable {name=Particles.Position, type=NiagaraPosition}}
      To Vector=$Particles.SkeletalMeshLocation.SampledPosition
     )
    )
   ) {enabled=false}
   &ParticleState()
   &ScaleSpriteSize(Scale Factor=&Vector2DFromFloat(Value=&FloatFromCurve(
      FloatCurve=DataInterface() {propertyValues {Curve {keys [
          {arriveTangent=18.9237, interpMode=RCIM_Cubic, leaveTangent=18.9237, tangentMode=RCTM_User, time=0, value=1}
          <at 0.1923 3.8141>
          {arriveTangent=-4.1284, interpMode=RCIM_Cubic, leaveTangent=-4.1284, time=0.6156, value=4.163}
          {arriveTangent=-7.3072, interpMode=RCIM_Cubic, leaveTangent=-7.3072, tangentMode=RCTM_User, time=0.9951, value=0.5}
         ]}}}
      Scale Curve=&Lerp_Float(A=1, B=1.5, Alpha=$System.Dive)
     )))
   &SolveForcesAndVelocity(Clamp Velocity=false, Limit Acceleration=false)
   &SubUVAnimation(Number Of Frames=64)
   &CameraOffset(Camera Offset Amount=&FloatFromCurve(FloatCurve=DataInterface() {propertyValues {Curve {keys [<at 0 -5>,<at 0.7093 2.1352>,<at 0.87 -2.1117>]}}}, Scale Curve=20))
  ]
  rendererClasses [&NiagaraSpriteRendererProperties]
  renderers [{rendererClass=&NiagaraSpriteRendererProperties}]
 }
 Flames_FlowLegs001 {
  emitterUpdateScript [
   &EmitterState(Life Cycle Mode=::NewEnumerator1, Loop Duration=1, UseLoopDelay=false)
   &SpawnRate(
    SpawnRate=&MakeCustomFloatFromBool(Bool=$User.bIsFrontEndPreview, True Float=30, False Float=&Lerp_Float(A=20, B=10, Alpha=$System.Dive))
    Use Spawn Probability=false
   )
  ]
  particleSpawnScript [
   &InitializeParticle(
    Lifetime=&UniformRangedFloat(Minimum=1.2, Maximum=1.8)
    Module.Write Color=true
    Module.Write Lifetime=true
    Module.Write Mass=true
    Module.Write Material Random=true
    Module.Write Position=false
    Module.Write Scale=false
    Module.Write SpriteRotation=true
    Module.Write SpriteSize=true
    Position=&SimulationPosition()
    Sprite Rotation=&UniformRangedFloat(Minimum=0, Maximum=350)
    Sprite Size=&Vector2DFromFloat(Value=&RandomRangeFloat(Minimum=3, Maximum=4))
   )
   /SetVariables(Particles.BoneFlowAngle:NiagaraFloat=&UniformRangedFloat(Minimum=0, Maximum=6.4)) {isSetParameters=true, module=SetVariables_#6}
   &FN_BoneFlow(
    Angle:NiagaraFloat=$Particles.BoneFlowAngle
    BoneID=HlslExpression() {hlslExpression="(Particles.UniqueID && 1) *3"}
    Radius:NiagaraFloat=&FloatFromCurve(FloatCurve=DataInterface() {propertyValues {Curve {keys [<at 0 1>,<at 1 0>]}}}, Scale Curve=10)
    SkeletalMesh:NiagaraDataInterfaceSkeletalMesh=$User.SkeletalMesh
    Write Position Attribute=true
   ) {enabled=false}
  ]
  particleUpdateScript [
   &ScaleColor(
    Scale Alpha=2
    Scale RGB=&MakeVectorFromLinearColorRGB(Color=&MultiplyLinearColor(
      Linear Color=&ColorFromCurve(ColorCurve=DataInterface() {propertyValues {
         AlphaCurve {keys [<at 0 1 RCIM_Cubic>,<at 1 1 RCIM_Cubic>]}
         BlueCurve {keys [<at 0.0285 0.01 RCIM_Cubic>, <at 0.5316 0.03>, {interpMode=RCIM_Cubic, leaveTangent=-0.1401, time=0.7835, value=0.05}]}
         GreenCurve {keys [<at 0.0285 0.2575 RCIM_Cubic>, <at 0.5316 0.1593>, {interpMode=RCIM_Cubic, leaveTangent=-0.7439, time=0.7835, value=0.3667}]}
         RedCurve {keys [<at 0.0285 1 RCIM_Cubic>,<at 0.5316 1>,<at 0.7835 1 RCIM_Cubic>]}
        }})
      Scale Factor=&MakeVector4_FromFloat(Float=&FloatFromCurve(
        FloatCurve=DataInterface() {propertyValues {Curve {keys [{interpMode=RCIM_Cubic, tangentMode=RCTM_User, time=0, value=0}, {interpMode=RCIM_Cubic, tangentMode=RCTM_User, time=1, value=2}]}}}
        Scale Curve=200
       ))
     ))
    ScaleA=true
    ScaleRGB=true
   ) {module=ScaleColor001}
   &FN_BoneFlow(
    Angle:NiagaraFloat=$Particles.BoneFlowAngle
    BoneID=HlslExpression() {hlslExpression="((Particles.UniqueID && 1) *1)+2"}
    Height:NiagaraFloat=&FloatFromCurve(FloatCurve=DataInterface() {propertyValues {Curve {keys [
         {arriveTangent=1.3633, interpMode=RCIM_Cubic, leaveTangent=1.3633, tangentMode=RCTM_User, time=0, value=0}
         {arriveTangent=0.3204, interpMode=RCIM_Cubic, leaveTangent=0.3204, tangentMode=RCTM_User, time=0.9897, value=1.1}
        ]}}})
    Interpolate Mesh Sampling=false
    Radius:NiagaraFloat=&FloatFromCurve(FloatCurve=DataInterface() {propertyValues {Curve {keys [
         {arriveTangent=0.8832, interpMode=RCIM_Cubic, time=0.1742, value=0.2546}
         <at 0.4784 0.2906>
         {arriveTangent=-0.2267, interpMode=RCIM_Cubic, leaveTangent=-0.2267, time=0.6089, value=0.5975}
         <at 0.7465 0.2298>
         {arriveTangent=-2.8131, interpMode=RCIM_Cubic, leaveTangent=-2.8131, tangentMode=RCTM_User, time=0.8737, value=0.065}
        ]}}})
    SkeletalMesh:NiagaraDataInterfaceSkeletalMesh=$User.SkeletalMesh
    Write Position Attribute=true
   ) {module=FN_BoneFlow002}
   /SetVariables(Particles.TestSplineBasedVector:Vector3f=$Output.FN_BoneFlow.SplineTangent) {isSetParameters=true, module=SetVariables_#7}
   &DynamicMaterialParameters(
    Index 0 Param 1=&FloatFromCurve(FloatCurve=DataInterface() {propertyValues {Curve {keys [
         {arriveTangent=2.125, interpMode=RCIM_Cubic, time=0, value=1}
         {arriveTangent=3.7313, interpMode=RCIM_Cubic, leaveTangent=3.7313, tangentMode=RCTM_User, time=0.1, value=0.15}
         <at 0.4 1 RCIM_Cubic>
        ]}}})
    Index 0 Param 2=1
    Index 0 Param 3=&FloatFromCurve(
     FloatCurve=DataInterface() {propertyValues {Curve {keys [
         {arriveTangent=-0.3008, interpMode=RCIM_Cubic, leaveTangent=-0.3008, tangentMode=RCTM_User, time=0, value=0}
         {interpMode=RCIM_Cubic, tangentMode=RCTM_User, time=0.5809, value=1}
        ]}}}
     Scale Curve=8
    )
    Index 0 Param 4=1
    Module.Param0WriteEnabled=true
    Module.Param0WriteEnabled1=true
    Module.Param1WriteEnabled=true
    Module.Param1WriteEnabled1=true
    Module.Param2WriteEnabled=true
    Module.Param2WriteEnabled1=true
    Module.Param3WriteEnabled=true
    Module.Param3WriteEnabled1=true
    Write Parameter Index 1=true
   ) {module=DynamicMaterialParameters001}
   &SkeletalMeshLocation(
    Apply Sampled Orientation to Mesh Particles=true
    Bone / Socket Index=$Particles.SpawnGroup
    Bone Sampling Mode=::NewEnumerator4
    Inherit Velocity=false
    Interpolate Mesh Sampling=false
    MaskSpawnGroup=false
    Skeletal Mesh=$User.SkeletalMesh
   ) {enabled=false}
   &AddVelocity(
    Scale Added Velocity=<100 100 100>
    Velocity=&RotateVectorByQuaternion(
     Vector To Rotate=<1 0 0>
     Quaternion=&FindQuatBetweenTwoVectors(
      From Vector=Linked() {linkedVariable {name=Particles.Position, type=NiagaraPosition}}
      To Vector=$Particles.SkeletalMeshLocation.SampledPosition
     )
    )
   ) {enabled=false}
   &ParticleState()
   &ScaleSpriteSize(Scale Factor=&Vector2DFromFloat(Value=&FloatFromCurve(
      FloatCurve=DataInterface() {propertyValues {Curve {keys [
          {arriveTangent=18.9237, interpMode=RCIM_Cubic, leaveTangent=18.9237, tangentMode=RCTM_User, time=0, value=1}
          <at 0.5509 3.8421>
          <at 0.7003 9.9984>
          {arriveTangent=-17.3587, interpMode=RCIM_Cubic, leaveTangent=-17.3586, tangentMode=RCTM_User, time=0.8294, value=5.9457}
          {arriveTangent=-7.3072, interpMode=RCIM_Cubic, leaveTangent=-7.3072, tangentMode=RCTM_User, time=0.9951, value=0.5}
         ]}}}
      Scale Curve=&Lerp_Float(A=1, B=1.5, Alpha=$System.Dive)
     )))
   &SolveForcesAndVelocity(Clamp Velocity=false, Limit Acceleration=false)
   &SubUVAnimation(Number Of Frames=&UniformRangedFloat(Minimum=20, Maximum=20)) {enabled=false}
   &CameraOffset(Camera Offset Amount=&FloatFromCurve(
     FloatCurve=DataInterface() {propertyValues {Curve {keys [<at 0.2 0 RCIM_Constant>,<at 0.3 0.5 RCIM_Constant>,<at 0.6 -0.645 RCIM_Constant>,<at 0.8548 0.3856 RCIM_Constant>]}}}
     Scale Curve=20
    ))
   /Niagara/Modules/Update/SubUV/V2/SubUVAnimation(
    Sub UV Play Rate:NiagaraFloat=&Lerp_Float(A=1, B=3, Alpha=$System.Dive)
    SubUV Animation Mode:ENiagara_SubUVLookupMode=::NewEnumerator3
   ) {module=SubUVAnimation001}
  ]
  rendererClasses [&NiagaraSpriteRendererProperties]
  renderers [{rendererClass=&NiagaraSpriteRendererProperties}]
 }
 Flames_Upward {
  emitterUpdateScript [
   &EmitterState(Life Cycle Mode=::NewEnumerator1, Loop Duration=1, UseLoopDelay=false)
   &SpawnRate(
    SpawnRate=&MakeCustomFloatFromBool(Bool=$User.bIsFrontEndPreview, True Float=160, False Float=&Lerp_Float(A=80, B=80, Alpha=$System.Dive))
    Use Spawn Probability=true
   )
  ]
  particleSpawnScript [
   &FN_DivingAttributes_v2(SkeletalMesh:NiagaraDataInterfaceSkeletalMesh=$User.SkeletalMesh)
   &FN_SetSpawnGroup_v2(Number of Spawn Groups=4)
   /SetVariables(Particles.DiveCLAMP:NiagaraFloat=&ClampFloat(
     Float=&FN_FrontEndFloatModifier(Gameplay=$Particles.Dive, Front End=&Multiply_Float(A=$Particles.Dive, B=10))
     Min=-1
    )) {isSetParameters=true, module=SetVariables_#1}
   &InitializeParticle(
    Lifetime=&Lerp_Float(
     A=&UniformRangedFloat(Minimum=0.4, Maximum=0.4)
     B=&UniformRangedFloat(Minimum=0.2, Maximum=0.4)
     Alpha=$Particles.DiveCLAMP
    )
    Module.Write Color=true
    Module.Write Lifetime=true
    Module.Write Mass=false
    Module.Write Material Random=true
    Module.Write Position=false
    Module.Write Scale=false
    Module.Write SpriteRotation=true
    Module.Write SpriteSize=true
    Position=&SimulationPosition()
    Sprite Rotation=&UniformRangedFloat(Minimum=50, Maximum=-360)
    Sprite Size=&Vector2DFromFloat(Value=&Lerp_Float(
      A=&UniformRangedFloat(Minimum=35, Maximum=55)
      B=&UniformRangedFloat(Minimum=70, Maximum=70)
      Alpha=$Particles.DiveCLAMP
     ))
   )
   &DynamicMaterialParameters(
    Index 0 Param 2=&UniformRangedFloat(Minimum=0, Maximum=1)
    Index 0 Param 3=1
    Index 0 Param 4=&UniformRangedFloat(Minimum=1, Maximum=0.5)
    Module.Param0WriteEnabled=false
    Module.Param1WriteEnabled=true
    Module.Param2WriteEnabled=false
    Module.Param3WriteEnabled=true
   ) {enabled=false, module=DynamicMaterialParameters002}
   &AddVelocity(Velocity=&Lerp_Vector(
     A=&UniformRangedVector(Minimum=<-25 0 0>, Maximum=<25 0 0>)
     B=&UniformRangedVector(Minimum=<-50 0 0>, Maximum=<50 0 0>)
     Alpha=$Particles.DiveCLAMP
    )) {enabled=false}
   &AddVelocity(Velocity=&Lerp_Vector(A=&RandomRangeVector(Minimum=<0 0 200>, Maximum=<1 1 300>), B=<0 0 500>, Alpha=$System.Dive)) {module=AddVelocity001}
   &SkeletalMeshLocation(
    Bone / Socket Index=HlslExpression() {hlslExpression=((Particles.SpawnGroup*3)+2)}
    Bone Sampling Mode=::NewEnumerator4
    Inherit Velocity=false
    "Inherit Velocity (Scale)"=0
    Interpolate Mesh Sampling=false
    MaskSpawnGroup=false
    Skeletal Mesh=$User.SkeletalMesh
   )
   &CameraOffset(Camera Offset Amount=-10)
   &SphereLocation(
    MaskSpawnGroup=false
    Override Seed=false
    Sphere Radius=2
    Surface Only=false
    UseNonUniformScale=false
    UseOffset=false
    UseSphereOrientation=false
   )
  ]
  particleUpdateScript [
   /SetVariables(Particles.DiveCLAMP:NiagaraFloat=&ClampFloat(Float=&FN_FrontEndFloatModifier(Gameplay=$Particles.Dive, Front End=&Multiply_Float(A=$Particles.Dive, B=10)))) {enabled=false, isSetParameters=true, module=SetVariables_#3}
   &ParticleState()
   &DynamicMaterialParameters(
    Index 0 Param 1=&FloatFromCurve(
     FloatCurve=DataInterface() {propertyValues {Curve {keys [{arriveTangent=2.125, interpMode=RCIM_Cubic, time=0, value=1.4}, <at 0.4 1 RCIM_Cubic>]}}}
     Scale Curve=&MakeCustomFloatFromBool(Bool=$User.bIsFrontEndPreview, True Float=0.98, False Float=1)
    )
    Index 0 Param 2=1
    Index 0 Param 3=&FloatFromCurve(
     FloatCurve=DataInterface() {propertyValues {Curve {keys [
         {arriveTangent=-0.3008, interpMode=RCIM_Cubic, leaveTangent=-0.3008, tangentMode=RCTM_User, time=0, value=0.2}
         {arriveTangent=4.1145, interpMode=RCIM_Cubic, leaveTangent=4.1145, tangentMode=RCTM_User, time=0.5809, value=1}
        ]}}}
     Scale Curve=6
    )
    Index 0 Param 4=&MakeCustomFloatFromBool(Bool=$User.bIsFrontEndPreview, True Float=2, False Float=1.2)
    Module.Param0WriteEnabled=true
    Module.Param0WriteEnabled1=true
    Module.Param1WriteEnabled=true
    Module.Param1WriteEnabled1=true
    Module.Param2WriteEnabled=true
    Module.Param2WriteEnabled1=true
    Module.Param3WriteEnabled=true
    Module.Param3WriteEnabled1=true
    Write Parameter Index 1=true
   ) {module=DynamicMaterialParameters001}
   &SkeletalMeshLocation(
    Bone / Socket Index=HlslExpression() {hlslExpression=((Particles.SpawnGroup*3)+2)}
    Bone Sampling Mode=::NewEnumerator4
    Inherit Velocity=false
    "Inherit Velocity (Scale)"=0
    MaskSpawnGroup=false
    Skeletal Mesh=$User.SkeletalMesh
   ) {enabled=false, module=SkeletalMeshLocation001}
   &Drag(Drag=2, Use Linear Drag=true, Use Rotational Drag=false)
   &SolveForcesAndVelocity(Clamp Velocity=false, Limit Acceleration=false)
   &ScaleSpriteSize(Scale Factor=&Vector2DFromFloat(Value=&FloatFromCurve(FloatCurve=DataInterface() {propertyValues {Curve {keys [
          {arriveTangent=-0.402, interpMode=RCIM_Cubic, leaveTangent=-0.402, tangentMode=RCTM_User, time=0, value=0.99}
          {arriveTangent=-0.2157, interpMode=RCIM_Cubic, leaveTangent=-0.2157, tangentMode=RCTM_User, time=1, value=0.5}
         ]}}})))
   &ScaleColor(
    Scale Alpha=&FloatFromCurve(FloatCurve=DataInterface() {propertyValues {Curve {keys [
         {arriveTangent=8.7245, interpMode=RCIM_Cubic, leaveTangent=8.7245, tangentMode=RCTM_User, time=0, value=0}
         <at 0.2 1 RCIM_Cubic>
         {arriveTangent=-0.5281, interpMode=RCIM_Cubic, leaveTangent=-0.5281, tangentMode=RCTM_User, time=1, value=0}
        ]}}})
    ScaleA=true
    ScaleRGB=true
   ) {enabled=false}
   &ScaleColor(
    Scale RGB=&MakeVectorFromLinearColorRGB(Color=&ColorFromCurve(ColorCurve=DataInterface() {propertyValues {
        AlphaCurve {keys [<at 0 1 RCIM_Cubic>,<at 1 1 RCIM_Cubic>]}
        BlueCurve {keys [<at 0.0828 0.03>,<at 1 0.03 RCIM_Cubic>]}
        GreenCurve {keys [<at 0.0828 0.2887>,<at 1 0.0462 RCIM_Cubic>]}
        RedCurve {keys [<at 0.0828 1>,<at 1 1 RCIM_Cubic>]}
       }}))
    ScaleA=true
    ScaleRGB=true
   ) {enabled=false, module=ScaleColor001}
   &ScaleColor(
    Scale Alpha=&FloatFromCurve(
     FloatCurve=DataInterface() {propertyValues {Curve {keys [<at 0 1>,<at 1 0>]}}}
     Scale Curve=&MakeCustomFloatFromBool(Bool=$User.bIsFrontEndPreview, True Float=0.2, False Float=1)
    )
    Scale RGB=&MakeVectorFromLinearColorRGB(Color=&MultiplyLinearColor(
      Linear Color=&MakeCustomLinearColorFromBool(
       Bool=$User.bIsFrontEndPreview
       True Color=&ColorFromCurve(ColorCurve=DataInterface() {propertyValues {
          AlphaCurve {keys [<at 0 1 RCIM_Cubic>,<at 1 1 RCIM_Cubic>]}
          BlueCurve {keys [<at 0.0038 0.2 RCIM_Cubic>, <at 0.1406 0>, <at 0.7874 0.03>, {interpMode=RCIM_Cubic, leaveTangent=0.0219, time=0.9159, value=0.1}]}
          GreenCurve {keys [<at 0.0038 0.5333 RCIM_Cubic>, <at 0.1406 0.5>, <at 0.7874 0.224>, {interpMode=RCIM_Cubic, leaveTangent=-1.0512, time=0.9159, value=0.1}]}
          RedCurve {keys [<at 0.0038 1 RCIM_Cubic>,<at 0.1406 2>,<at 0.7874 1>,<at 0.9159 5 RCIM_Cubic>]}
         }})
       False Color=&ColorFromCurve(ColorCurve=DataInterface() {propertyValues {
          AlphaCurve {keys [<at 0 1 RCIM_Cubic>,<at 1 1 RCIM_Cubic>]}
          BlueCurve {keys [<at 0.0038 0.2 RCIM_Cubic>, <at 0.0742 0.02>, <at 0.386 0.03>, {interpMode=RCIM_Cubic, leaveTangent=0.0219, time=0.7892, value=0.1}]}
          GreenCurve {keys [<at 0.0038 0.5333 RCIM_Cubic>, <at 0.0742 0.5917>, <at 0.386 0.1108>, {interpMode=RCIM_Cubic, leaveTangent=-1.0512, time=0.7892, value=0.1}]}
          RedCurve {keys [<at 0.0038 1 RCIM_Cubic>,<at 0.0742 1>,<at 0.386 1>,<at 0.7892 5 RCIM_Cubic>]}
         }})
      )
      Scale Factor=&MakeVector4_FromFloat(Float=&FloatFromCurve(
        FloatCurve=DataInterface() {propertyValues {Curve {keys [{interpMode=RCIM_Cubic, tangentMode=RCTM_User, time=0, value=1}, {interpMode=RCIM_Cubic, tangentMode=RCTM_User, time=1, value=2}]}}}
        Scale Curve=8
       ))
     ))
    ScaleA=true
    ScaleRGB=true
   ) {module=ScaleColor002}
   &SubUVAnimation(
    Sub UV Play Rate:NiagaraFloat=&Lerp_Float(A=1, B=3, Alpha=$System.Dive)
    SubUV Animation Mode:ENiagara_SubUVLookupMode=::NewEnumerator3
   ) {module=SubUVAnimation001}
   &ScaleSpriteSize(Scale Factor=<1.05 1.05>) {enabled=false, module=ScaleSpriteSize001}
  ]
  rendererClasses [&NiagaraSpriteRendererProperties]
  renderers [{rendererClass=&NiagaraSpriteRendererProperties}]
 }
 Lights {
  emitterSpawnScript [
   &FN_PreserveSkydiveUserVariables() {module=PreserveSkydiveUserVariables}
   &FN_PartitionAttributes(Particles Per Partition=10, Partitions=4)
  ]
  emitterUpdateScript [
   &EmitterState()
   &SpawnRate(SpawnRate=7, Use Spawn Probability=false)
  ]
  particleSpawnScript [
   &InitializeParticle(
    Color=<rgba 30 6.5863 0 1>
    Lifetime=0.5
    Module.Write Color=true
    Module.Write Lifetime=true
    Module.Write Mass=false
    Module.Write Material Random=true
    Module.Write Position=false
    Module.Write Scale=false
    Module.Write SpriteRotation=false
    Module.Write SpriteSize=true
    Position=&SimulationPosition()
    Sprite Size=<2 2>
   )
   &FN_Particle_Partition(Pattern:ENiagaraPartitionWaveForm=::NewEnumerator1, Seed:NiagaraInt32=$Emitter.RandomSeed)
   /SetVariables(Particles.LightRadius=30) {isSetParameters=true, module=SetVariables_#4}
   /SetVariables(Particles.BoneFlowAngle:NiagaraFloat=&UniformRangedFloat(Minimum=0, Maximum=6.4)) {isSetParameters=true, module=SetVariables_#5}
  ]
  particleUpdateScript [
   &SampleSkeletalMeshSkeleton(Mesh:NiagaraDataInterfaceSkeletalMesh=$User.SkeletalMesh, Specific Bone Index:NiagaraInt32=$Particles.SpawnGroup)
   &ParticleState()
   &FN_BoneFlow(
    Angle:NiagaraFloat=$Particles.BoneFlowAngle
    BoneID:NiagaraInt32=&Add_Integer(A=$Particles.Partition.ID, B=0)
    Height:NiagaraFloat=&FloatFromCurve(FloatCurve=DataInterface() {propertyValues {Curve {keys [
         {arriveTangent=0.9481, interpMode=RCIM_Cubic, leaveTangent=0.9481, tangentMode=RCTM_User, time=0, value=0.5}
         {arriveTangent=0.3204, interpMode=RCIM_Cubic, leaveTangent=0.3204, tangentMode=RCTM_User, time=0.9897, value=1.2}
        ]}}})
    Interpolate Mesh Sampling=false
    Radius:NiagaraFloat=&FloatFromCurve(
     FloatCurve=DataInterface() {propertyValues {Curve {keys [
         <at 0.23 0.3473 RCIM_Cubic>
         {arriveTangent=1.0971, interpMode=RCIM_Cubic, leaveTangent=1.0971, time=0.6661, value=0.6553}
         {arriveTangent=-2.8131, interpMode=RCIM_Cubic, leaveTangent=-2.8131, tangentMode=RCTM_User, time=0.863, value=1.0417}
        ]}}}
     Scale Curve=5
    )
    SkeletalMesh:NiagaraDataInterfaceSkeletalMesh=$User.SkeletalMesh
    Write Position Attribute=true
   )
  ]
  rendererClasses [&NiagaraLightRendererProperties, &NiagaraSpriteRendererProperties]
  renderers [
   {rendererClass=&NiagaraLightRendererProperties, rendererIndex=0}
   {rendererClass=&NiagaraSpriteRendererProperties, rendererIndex=1}
  ]
 }
 Mesh_Dive_Streaks {
  bEnabled=false
  emitterUpdateScript [
   &EmitterState(Life Cycle Mode=::NewEnumerator1, Loop Duration=1, UseLoopDelay=false)
   &SpawnRate(SpawnRate=15, Use Spawn Probability=false)
   &SpawnRate(SpawnGroup=1, SpawnRate=4.5, Use Spawn Probability=false) {enabled=false, module=SpawnRate001}
  ]
  particleSpawnScript [
   &FN_DivingAttributes_v2(Accumulate Dive=0.1, Accumulate Rest=-0.1, SkeletalMesh:NiagaraDataInterfaceSkeletalMesh=$User.SkeletalMesh)
   &FN_SetSpawnGroup_v2(Number of Spawn Groups=4)
   /SetVariables(Particles.DiveCLAMP:NiagaraFloat=&ClampFloat(Float=&FN_FrontEndFloatModifier(Gameplay=$Particles.Dive, Front End=&Multiply_Float(A=$Particles.Dive, B=10)))) {isSetParameters=true, module=SetVariables_#1}
   /SetVariables(Particles.MaterialRandom:NiagaraFloat=&UniformRangedFloat(Minimum=0, Maximum=1)) {isSetParameters=true, module=SetVariables_#2}
   &InitializeParticle(
    Color=&LerpLinearColors(
     StartColor=<rgba 50 0.8333 0 0>
     EndColor=&LerpLinearColors(
      StartColor=<rgba 1 0.0167 0 0.4>
      EndColor=<rgba 50 8.8333 2.5 1>
      LerpFactor=&UniformAOrBFloat(A=0.5, B=0.7)
     )
     LerpFactor=$Particles.DiveCLAMP
    )
    Lifetime=&Lerp_Float(
     A=&UniformRangedFloat(Minimum=0.5, Maximum=0.7)
     B=&UniformRangedFloat(Minimum=0.8, Maximum=1)
     Alpha=$Particles.DiveCLAMP
    )
    Material Random:NiagaraFloat=$Particles.MaterialRandom
    Mesh Scale=&Multiply_Vector(A=<0.4 0.125 0.125>, B=&Lerp_Vector(A=<1 1 1>, B=<2 1 1>, Alpha=$Particles.DiveCLAMP))
    Module.Write Color=true
    Module.Write Lifetime=true
    Module.Write Mass=true
    Module.Write Material Random=true
    Module.Write Position=false
    Module.Write Scale=true
    Module.Write SpriteRotation=true
    Module.Write SpriteSize=false
    Sprite Rotation=&UniformRangedFloat(Minimum=-0.2, Maximum=0.2)
    Sprite Size=<5 5>
   )
   &SkeletalMeshLocation(
    Apply To Particle Position=false
    Bone / Socket Index=&SpawnGroup()
    Bone Sampling Mode=::NewEnumerator4
    Inherit Velocity=false
    MaskSpawnGroup=false
    Sampled Position Offset=<0 5 0>
    Skeletal Mesh=$User.SkeletalMesh
   ) {module=SkeletalMeshLocation001}
   &InitialMeshOrientation(
    Orientation Vector=<0 0 -1>
    Rotation=&UniformRangedVector(Minimum=<0 0 0>, Maximum=<1 0 0>)
    Use Orientation Vector=true
    Use Rotation Vector=true
   )
  ]
  particleUpdateScript [
   /SetVariables(Particles.DiveCLAMP:NiagaraFloat=&ClampFloat(Float=&FN_FrontEndFloatModifier(Gameplay=$Particles.Dive, Front End=&Multiply_Float(A=$Particles.Dive, B=10)))) {isSetParameters=true, module=SetVariables_#3}
   &ParticleState()
   &SkeletalMeshLocation(
    Bone / Socket Index=HlslExpression() {hlslExpression=((Particles.SpawnGroup*3)+2)}
    Bone Sampling Mode=::NewEnumerator4
    Inherit Velocity=false
    MaskSpawnGroup=false
    Sampled Position Offset Coordinate Space:ENiagaraMeshTransforms=::NewEnumerator2
    Skeletal Mesh=$User.SkeletalMesh
   )
   &UpdateMeshOrientation(
    Delta Time:NiagaraFloat=&Multiply_Float(A=$Engine.DeltaTime, B=&Multiply_Float(A=5, B=$Particles.NormalizedAge))
    Rotation Rate:NiagaraFloat=$Particles.Initial.SpriteRotation
    Rotation Vector:Vector3f=<0 0 1>
   ) {enabled=false}
   &ScaleMeshSize(Scale Factor=&Multiply_Vector(
     A=&Lerp_Vector(
      A=&MakeVector(
       X=&FloatFromCurve(FloatCurve=DataInterface() {propertyValues {Curve {keys [<at 0 0.1>,<at 1 2>]}}})
       Y=&FloatFromCurve(FloatCurve=DataInterface() {propertyValues {Curve {keys [
            <at 0 1 RCIM_Cubic>
            {arriveTangent=-1.1222, interpMode=RCIM_Cubic, leaveTangent=-2.1367, tangentMode=RCTM_User, time=0.9, value=0.5}
            <at 1 0.05>
           ]}}})
       Z=&FloatFromCurve(FloatCurve=DataInterface() {propertyValues {Curve {keys [
            <at 0 1 RCIM_Cubic>
            {arriveTangent=-1.1222, interpMode=RCIM_Cubic, leaveTangent=-2.1367, tangentMode=RCTM_User, time=0.9, value=0.5}
            <at 1 0.05>
           ]}}})
      )
      B=&MakeVector(
       X=&FloatFromCurve(FloatCurve=DataInterface() {propertyValues {Curve {keys [<at 0 1>,<at 1 2>]}}})
       Y=&FloatFromCurve(FloatCurve=DataInterface() {propertyValues {Curve {keys [
            {arriveTangent=-1.5003, interpMode=RCIM_Cubic, leaveTangent=-0.2737, tangentMode=RCTM_User, time=0, value=1}
            {arriveTangent=-2.2772, interpMode=RCIM_Cubic, leaveTangent=-2.2772, tangentMode=RCTM_User, time=1, value=0}
           ]}}})
       Z=&FloatFromCurve(FloatCurve=DataInterface() {propertyValues {Curve {keys [
            {arriveTangent=-1.5003, interpMode=RCIM_Cubic, leaveTangent=-0.2737, tangentMode=RCTM_User, time=0, value=1}
            {arriveTangent=-2.2772, interpMode=RCIM_Cubic, leaveTangent=-2.2772, tangentMode=RCTM_User, time=1, value=0}
           ]}}})
      )
      Alpha=$Particles.DiveCLAMP
     )
     B=&VectorFromFloat(Value=&FloatFromCurve(FloatCurve=DataInterface() {propertyValues {Curve {keys [
           <at 0 0 RCIM_Constant>
           {arriveTangent=5.5862, interpMode=RCIM_Cubic, leaveTangent=5.5862, tangentMode=RCTM_User, time=0.3, value=0}
           <at 0.7 1>
           <at 1 1>
          ]}}}))
    )) {module=ScaleMeshSize001}
   &ScaleColor(
    Scale Alpha=&FloatFromCurve(FloatCurve=DataInterface() {propertyValues {Curve {keys [
         {arriveTangent=8.7245, interpMode=RCIM_Cubic, leaveTangent=8.7245, tangentMode=RCTM_User, time=0, value=0}
         <at 0.45 1 RCIM_Cubic>
         {arriveTangent=-0.5281, interpMode=RCIM_Cubic, leaveTangent=-0.5281, tangentMode=RCTM_User, time=1, value=0}
        ]}}})
    ScaleA=true
    ScaleRGB=true
   )
   &DynamicMaterialParameters(
    Index 0 Param 1=&Lerp_Float(A=0, B=0.5, Alpha=$Particles.DiveCLAMP)
    Module.Param0WriteEnabled=true
    Module.Param1WriteEnabled=false
    Module.Param2WriteEnabled=false
    Module.Param3WriteEnabled=false
   )
  ]
  rendererClasses [&NiagaraMeshRendererProperties]
  renderers [{rendererClass=&NiagaraMeshRendererProperties, rendererIndex=0}]
 }
 Ribbon_FireStreaks {
  bEnabled=false
  emitterUpdateScript [
   &EmitterState()
   /Game/Effects/Niagara/Modules/Ribbons/FN_SpawnRibbon_v3(
    Number of Ribbons=4
    Skeleton Mesh:NiagaraDataInterfaceSkeletalMesh=$User.SkelMeshLimbs
    SpawnRate:NiagaraFloat=&Lerp_Float(A=10, B=30, Alpha=$System.Dive)
   ) {enabled=false}
   &FN_SpawnRibbon_v2(
    Number of Ribbons=4
    Spawn Mode=::NewEnumerator0
    SpawnRate:NiagaraFloat=&MakeCustomFloatFromBool(Bool=$User.bIsFrontEndPreview, True Float=20, False Float=&Lerp_Float(A=10, B=30, Alpha=$System.Dive))
   )
   &FN_SpawnRibbon_v2(
    Number of Ribbons=4
    Spawn Mode=::NewEnumerator0
    SpawnRate:NiagaraFloat=&Lerp_Float(A=0, B=30, Alpha=$System.Dive)
   ) {enabled=false, module=FN_SpawnRibbon_v2001}
  ]
  particleSpawnScript [
   &FN_SetSpawnGroup()
   /Game/Effects/Niagara/Modules/Ribbons/FN_InitializeRibbon_v5(
    Enable Skeletal Position=true
    Lifetime:NiagaraFloat=&MakeCustomFloatFromBool(Bool=$User.bIsFrontEndPreview, True Float=0.3, False Float=&Lerp_Float(A=0.2, B=0.08, Alpha=$System.Dive))
    Ribbon Width=18
    SkeletalMesh:NiagaraDataInterfaceSkeletalMesh=$User.SkelMeshLimbs
   )
   &AddVelocity(Velocity=&FN_FrontEndVector3Modifier(Gameplay=<0 0 800>, Front End=<0 0 1200>)) {module=AddVelocity001}
   /SetVariables(Particles.Color=&LerpLinearColors(
     StartColor=&ColorFromCurve(
      ColorCurve=DataInterface() {propertyValues {
        AlphaCurve {keys [<at 0 1 RCIM_Cubic>]}
        BlueCurve {keys [<at 0 1 RCIM_Cubic>,<at 1 1 RCIM_Cubic>]}
        GreenCurve {keys [<at 0 1 RCIM_Cubic>,<at 1 1 RCIM_Cubic>]}
        RedCurve {keys [<at 0 1 RCIM_Cubic>,<at 1 1 RCIM_Cubic>]}
       }}
      CurveIndex=&RandomRangeFloat(Minimum=0, Maximum=1)
     )
     EndColor=&ColorFromCurve(
      ColorCurve=DataInterface() {propertyValues {
        AlphaCurve {keys [<at 0 1 RCIM_Cubic>]}
        BlueCurve {keys [<at 0 1 RCIM_Cubic>,<at 1 1 RCIM_Cubic>]}
        GreenCurve {keys [<at 0 1 RCIM_Cubic>,<at 1 1 RCIM_Cubic>]}
        RedCurve {keys [<at 0 1 RCIM_Cubic>,<at 1 1 RCIM_Cubic>]}
       }}
      CurveIndex=&RandomRangeFloat(Minimum=0, Maximum=1)
     )
     LerpFactor=&MakeFloatFromInt(Int=$Particles.SpawnGroup)
    )) {isSetParameters=true, module=SetVariables_#8}
   /SetVariables(Particles.MaterialRandom:NiagaraFloat=&RandomRangeFloat(Minimum=0, Maximum=1)) {isSetParameters=true, module=SetVariables_#9}
   /SetVariables() {isSetParameters=true, module=SetVariables_#10}
   &SphereLocation(
    MaskSpawnGroup=false
    Override Seed=false
    Sphere Radius=1
    Surface Only=false
    UseNonUniformScale=false
    UseOffset=false
    UseSphereOrientation=false
   )
  ]
  particleUpdateScript [
   &ScaleVelocity(Velocity Scale=&MakeVector(
     X=0
     Y=0
     Z=&Multiply_Float(
      A=0.9
      B=&FloatFromCurve(
       FloatCurve=DataInterface() {propertyValues {Curve {keys [<at 0 0.9>,<at 1 3>]}}}
       CurveIndex=&Divide_Float(A=&VectorLength(Vector=$Engine.Owner.Velocity), B=9000)
      )
     )
    )) {module=ScaleVelocity001}
   &SolveForcesAndVelocity(Clamp Velocity=false, Limit Acceleration=false)
   /Game/Effects/Niagara/Modules/Ribbons/FN_UpdateRibbon_v6(Enable Skeletal Position=true, SkeletalMesh:NiagaraDataInterfaceSkeletalMesh=$User.SkelMeshLimbs)
   &DynamicMaterialParameters(
    Index 0 Param 1=&MaskFloatBySpawnGroup(Masked Float=0, Passthrough Float=0.5)
    Index 0 Param 2=&FloatFromCurve(FloatCurve=DataInterface() {propertyValues {Curve {keys [
         {arriveTangent=-0.3062, interpMode=RCIM_Cubic, leaveTangent=-0.3062, tangentMode=RCTM_User, time=0, value=0.5}
         {arriveTangent=0.0167, interpMode=RCIM_Cubic, leaveTangent=0.0167, tangentMode=RCTM_Break, time=1, value=0}
        ]}}})
    Index 0 Param 3=1
    Index 0 Param 4=1
    Module.Param0WriteEnabled=true
    Module.Param1WriteEnabled=true
    Module.Param2WriteEnabled=true
    Module.Param3WriteEnabled=true
   ) {enabled=false}
   &ScaleRibbonWidth(Ribbon Width Scale=&FloatFromCurve(
     FloatCurve=DataInterface() {propertyValues {Curve {keys [{interpMode=RCIM_Cubic, tangentMode=RCTM_User, time=0, value=1}, {interpMode=RCIM_Cubic, tangentMode=RCTM_User, time=1, value=0.5}]}}}
     Scale Curve=1.5
    ))
   &ScaleColor(
    Scale Alpha=&Multiply_Float(
     A=&FNTimeOfDayScalar(Night Scale=1.5, Morning Scale=2, Evening Scale=2)
     B=&FloatFromCurve(
      FloatCurve=DataInterface() {propertyValues {Curve {keys [
          {arriveTangent=0.8004, interpMode=RCIM_Cubic, leaveTangent=0.8004, tangentMode=RCTM_User, time=0.1, value=0}
          <at 0.4 1.025>
          {interpMode=RCIM_Cubic, tangentMode=RCTM_User, time=0.5, value=0}
         ]}}}
      Scale Curve=&Lerp_Float(A=0, B=0.15, Alpha=$System.Dive)
     )
    )
    Scale RGB=&Multiply_Vector(
     A=&MakeVectorFromLinearColorRGB(Color=&ColorFromCurve(ColorCurve=DataInterface() {propertyValues {
         AlphaCurve {keys [<at 0 1 RCIM_Cubic>,<at 1 1 RCIM_Cubic>]}
         BlueCurve {keys [<at 0 0.5 RCIM_Cubic>,<at 0.0927 0>,<at 0.155 0>,<at 0.4954 0>]}
         GreenCurve {keys [<at 0 0.9167 RCIM_Cubic>,<at 0.0927 0.5833>,<at 0.155 0.25>,<at 0.4954 0.0208>]}
         RedCurve {keys [<at 0 1 RCIM_Cubic>,<at 0.0927 1>,<at 0.155 1>,<at 0.4954 0.05>]}
        }}))
     B=&VectorFromFloat(Value=10)
    )
    ScaleA=true
    ScaleRGB=true
   )
  ]
  rendererClasses [&NiagaraRibbonRendererProperties]
  renderers [{rendererClass=&NiagaraRibbonRendererProperties}]
 }
 Ribbon_Smoke {
  emitterUpdateScript [
   &EmitterState()
   &FN_SpawnRibbon_v3(
    Number of Ribbons=4
    Skeleton Mesh:NiagaraDataInterfaceSkeletalMesh=$User.SkelMeshLimbs
    SpawnRate:NiagaraFloat=&Lerp_Float(A=5, B=30, Alpha=$System.Dive)
   )
  ]
  particleSpawnScript [
   &FN_SetSpawnGroup()
   &FN_InitializeRibbon_v5(
    Enable Skeletal Position=true
    Lifetime:NiagaraFloat=&Lerp_Float(A=0.4, B=0.25, Alpha=$System.Dive)
    Ribbon Width=10
    SkeletalMesh:NiagaraDataInterfaceSkeletalMesh=$User.SkelMeshLimbs
   )
   &AddVelocity(Velocity=&FN_FrontEndVector3Modifier(Gameplay=<0 0 800>, Front End=<0 0 1200>)) {module=AddVelocity001}
   /SetVariables(Particles.Color=&LerpLinearColors(
     StartColor=&ColorFromCurve(
      ColorCurve=DataInterface() {propertyValues {
        AlphaCurve {keys [<at 0 1 RCIM_Cubic>]}
        BlueCurve {keys [<at 0 1 RCIM_Cubic>,<at 1 1 RCIM_Cubic>]}
        GreenCurve {keys [<at 0 1 RCIM_Cubic>,<at 1 1 RCIM_Cubic>]}
        RedCurve {keys [<at 0 1 RCIM_Cubic>,<at 1 1 RCIM_Cubic>]}
       }}
      CurveIndex=&RandomRangeFloat(Minimum=0, Maximum=1)
     )
     EndColor=&ColorFromCurve(
      ColorCurve=DataInterface() {propertyValues {
        AlphaCurve {keys [<at 0 1 RCIM_Cubic>]}
        BlueCurve {keys [<at 0 1 RCIM_Cubic>,<at 1 1 RCIM_Cubic>]}
        GreenCurve {keys [<at 0 1 RCIM_Cubic>,<at 1 1 RCIM_Cubic>]}
        RedCurve {keys [<at 0 1 RCIM_Cubic>,<at 1 1 RCIM_Cubic>]}
       }}
      CurveIndex=&RandomRangeFloat(Minimum=0, Maximum=1)
     )
     LerpFactor=&MakeFloatFromInt(Int=$Particles.SpawnGroup)
    )) {isSetParameters=true, module=SetVariables_#8}
   /SetVariables(Particles.MaterialRandom:NiagaraFloat=&RandomRangeFloat(Minimum=0, Maximum=1)) {isSetParameters=true, module=SetVariables_#9}
   /SetVariables() {isSetParameters=true, module=SetVariables_#10}
  ]
  particleUpdateScript [
   &ScaleVelocity(Velocity Scale=&MakeVector(
     X=0
     Y=0
     Z=&Multiply_Float(
      A=0.9
      B=&FloatFromCurve(
       FloatCurve=DataInterface() {propertyValues {Curve {keys [<at 0 0.9>,<at 1 3>]}}}
       CurveIndex=&Divide_Float(A=&VectorLength(Vector=$Engine.Owner.Velocity), B=9000)
      )
     )
    )) {module=ScaleVelocity001}
   &SolveForcesAndVelocity(Clamp Velocity=false, Limit Acceleration=false)
   &FN_UpdateRibbon_v6(Enable Skeletal Position=true, SkeletalMesh:NiagaraDataInterfaceSkeletalMesh=$User.SkelMeshLimbs)
   &DynamicMaterialParameters(
    Index 0 Param 1=&MaskFloatBySpawnGroup(Masked Float=0, Passthrough Float=0.5)
    Index 0 Param 2=&FloatFromCurve(FloatCurve=DataInterface() {propertyValues {Curve {keys [
         {arriveTangent=-0.3062, interpMode=RCIM_Cubic, leaveTangent=-0.3062, tangentMode=RCTM_User, time=0, value=0.5}
         {arriveTangent=0.0167, interpMode=RCIM_Cubic, leaveTangent=0.0167, tangentMode=RCTM_Break, time=1, value=0}
        ]}}})
    Index 0 Param 3=1
    Index 0 Param 4=1
    Module.Param0WriteEnabled=true
    Module.Param1WriteEnabled=true
    Module.Param2WriteEnabled=true
    Module.Param3WriteEnabled=true
   ) {enabled=false}
   &ScaleRibbonWidth(Ribbon Width Scale=&FloatFromCurve(FloatCurve=DataInterface() {propertyValues {Curve {keys [
         {arriveTangent=-0.0063, interpMode=RCIM_Cubic, leaveTangent=-0.0063, tangentMode=RCTM_User, time=0, value=1.25}
         {arriveTangent=1.9375, interpMode=RCIM_Cubic, leaveTangent=1.9375, time=0.2, value=1.15}
         {arriveTangent=2.35, interpMode=RCIM_Cubic, leaveTangent=2.35, time=0.4, value=2.025}
         <at 0.7 2.325>
        ]}}}))
   &ScaleColor(
    Scale Alpha=&Multiply_Float(
     A=&FNTimeOfDayScalar(Night Scale=1.5, Morning Scale=2, Evening Scale=2)
     B=&FloatFromCurve(
      FloatCurve=DataInterface() {propertyValues {Curve {keys [
          {arriveTangent=-2.4914, interpMode=RCIM_Cubic, leaveTangent=-2.4914, tangentMode=RCTM_User, time=0, value=1}
          {interpMode=RCIM_Cubic, tangentMode=RCTM_User, time=1, value=0}
         ]}}}
      Scale Curve=&Lerp_Float(A=1, B=10, Alpha=$System.Dive)
     )
    )
    Scale RGB=&Multiply_Vector(
     A=&MakeVectorFromLinearColorRGB(Color=&ColorFromCurve(ColorCurve=DataInterface() {propertyValues {
         AlphaCurve {keys [<at 0 1 RCIM_Cubic>,<at 1 1 RCIM_Cubic>]}
         BlueCurve {keys [<at 0 0>,<at 0.085 0.25 RCIM_Cubic>]}
         GreenCurve {keys [<at 0 1.2152>,<at 0.085 0.25 RCIM_Cubic>]}
         RedCurve {keys [<at 0 3>,<at 0.085 0.25 RCIM_Cubic>]}
        }}))
     B=&VectorFromFloat(Value=1)
    )
    ScaleA=true
    ScaleRGB=true
   )
  ]
  rendererClasses [&NiagaraRibbonRendererProperties]
  renderers [{rendererClass=&NiagaraRibbonRendererProperties}]
 }
 Ribbons {
  bEnabled=false
  emitterUpdateScript [
   &EmitterState(Life Cycle Mode=::NewEnumerator1, Loop Duration=0, UseLoopDelay=false)
   /Game/Effects/Niagara/Modules/Skydiving/FN_DivingAttributes/FN_EmitterDivingAttributes_v1(Bone Index=1, SkeletalMesh:NiagaraDataInterfaceSkeletalMesh=$User.SkelMeshLimbs)
   &FN_SpawnRibbon_v2(Number of Ribbons=4, Spawn Mode=::NewEnumerator0, SpawnRate=64) {enabled=false}
   &FN_SpawnRibbon_v2(
    Number of Ribbons=4
    Spawn Mode=::NewEnumerator0
    SpawnRate:NiagaraFloat=&Lerp_Float(A=0, B=30, Alpha=$System.Dive)
   ) {module=FN_SpawnRibbon_v2001}
  ]
  particleSpawnScript [
   &FN_SetSpawnGroup()
   &FN_InitializeRibbon_v3(Custom Offset=true, Enable Skeletal Position=true, Lifetime=0.3, Local Vector Offset:Vector3f=<0 1 0>, Ribbon Width=11) {enabled=false}
   &FN_InitializeRibbon_v3(
    Enable Skeletal Position=true
    Lifetime:NiagaraFloat=&Lerp_Float(A=0.1, B=0.3, Alpha=$System.Dive)
    Ribbon Width:NiagaraFloat=&Lerp_Float(A=20, B=10, Alpha=$System.Dive)
    Write Material Random=false
   ) {module=FN_InitializeRibbon_v3001}
   &AddVelocity(
    Scale Added Velocity=&Lerp_Vector(A=<1 1 1>, B=<4 4 1>, Alpha=$System.Dive)
    Velocity=&FN_FrontEndVector3Modifier(Gameplay=<0 0 999>, Front End=<0 -333 333>)
   )
   &DynamicMaterialParameters(
    Index 0 Param 1=&MaskFloatBySpawnGroup(Masked Float=0.333, Passthrough Float=0.666)
    Module.Param0WriteEnabled=true
    Module.Param1WriteEnabled=false
    Module.Param2WriteEnabled=false
    Module.Param3WriteEnabled=false
   ) {enabled=false}
   &SkeletalMeshLocation(
    Apply Sampled Orientation to Mesh Particles=true
    Bone / Socket Index=$Particles.SpawnGroup
    Bone Sampling Mode=::NewEnumerator4
    Inherit Velocity=false
    Interpolate Mesh Sampling=false
    MaskSpawnGroup=false
    Sampled Position Offset=&MaskVectorBySpawnGroup(Masked Vector=<10 0 0>, Passthrough Vector=<-10 0 0>)
    Sampled Position Offset Coordinate Space:ENiagaraMeshTransforms=::NewEnumerator3
    Skeletal Mesh=$User.SkelMeshLimbs
   )
   &FN_FlipRibbonV(Offset=1) {enabled=false}
  ]
  particleUpdateScript [
   &InheritSourceMovement(Applied Position Scale=<0.4 0.4 1>, ApplyForce=false, ApplyPosition=true, ApplyPositionFalloff=false)
   &CurlNoiseForce(
    Noise Strength=300
    Pan Noise=false
    Sample Position:Vector3f=&TransformPosition(Position=Linked() {linkedVariable {name=Particles.Position, type=NiagaraPosition}}, Source Space=::World, Destination Space=::Local)
   ) {enabled=false}
   &SolveForcesAndVelocity(Clamp Velocity=false, Limit Acceleration=false)
   &FN_RibbonMaterial_v2(
    Color:LinearColor=&MakeLinearColorFromVectorAndFloat(
     "Vector (RGB)"=&MakeVectorFromLinearColorRGB(Color=&MultiplyLinearColor(
       Linear Color=&ColorFromCurve(ColorCurve=DataInterface() {propertyValues {
          AlphaCurve {keys [<at 0 1 RCIM_Cubic>,<at 0.2749 1 RCIM_Cubic>]}
          BlueCurve {keys [<at 0 0.2 RCIM_Cubic>,<at 0.0514 0.2>,<at 0.2019 0.0677>]}
          GreenCurve {keys [<at 0 0.2 RCIM_Cubic>,<at 0.0514 0.3333>,<at 0.2019 0.0677>]}
          RedCurve {keys [<at 0 1 RCIM_Cubic>,<at 0.0514 1>,<at 0.2019 0.0677>]}
         }})
       Scale Factor=&MakeVector4_FromFloat(Float=1)
      ))
     "Float (Alpha)"=2
    )
    Color Pan U Speed=1
    Erosion Amount=1
    Erosion Power=2
    Macro Distortion Amount:NiagaraFloat=&FloatFromCurve(FloatCurve=DataInterface() {propertyValues {Curve {keys [<at 0.0051 0.2>,<at 0.2115 0.4>,<at 0.4914 0.115>,<at 1 0>]}}})
    Macro Pan Offset=1
    Macro Pan Speed=2
    Macro Scale U=2
    Opacity Erosion:NiagaraFloat=&FloatFromCurve(FloatCurve=DataInterface() {propertyValues {Curve {keys [
         {interpMode=RCIM_Cubic, tangentMode=RCTM_User, time=0, value=0}
         {arriveTangent=1, interpMode=RCIM_Cubic, leaveTangent=1, tangentMode=RCTM_User, time=0.7307, value=0.1673}
         <at 0.9872 0.5>
        ]}}})
    Opacity Pan U Offset=1
    Opacity Pan U Speed=2
    Opacity Power=2
   )
   /Game/Effects/Niagara/Modules/Materials/FN_RibbonMaterial/FN_Ribbon_v3/FN_RibbonMaterial() {enabled=false}
   &FN_UpdateRibbon(Detach Locked Particle=true, Enable Skeletal Position=true)
  ]
  rendererClasses [&NiagaraRibbonRendererProperties]
  renderers [
   {rendererClass=&NiagaraRibbonRendererProperties}
   {rendererClass=&NiagaraRibbonRendererProperties, rendererIndex=1}
   {rendererClass=&NiagaraRibbonRendererProperties, rendererIndex=2}
  ]
 }
 Ribbons001 {
  bEnabled=false
  emitterUpdateScript [
   &EmitterState(Life Cycle Mode=::NewEnumerator1, Loop Duration=0, UseLoopDelay=false)
   &FN_EmitterDivingAttributes_v1(Bone Index=1, SkeletalMesh:NiagaraDataInterfaceSkeletalMesh=$User.SkelMeshLimbs)
   &FN_SpawnRibbon_v2(Number of Ribbons=4, Spawn Mode=::NewEnumerator0, SpawnRate=64)
  ]
  particleSpawnScript [
   &FN_SetSpawnGroup()
   &FN_InitializeRibbon_v3(Custom Offset=true, Enable Skeletal Position=true, Lifetime=0.3, Local Vector Offset:Vector3f=<0 1 0>, Ribbon Width=11)
   &AddVelocity(
    Scale Added Velocity=&Lerp_Vector(A=<1 1 1>, B=<4 4 1>, Alpha=$System.Dive)
    Velocity=&FN_FrontEndVector3Modifier(Gameplay=<0 0 999>, Front End=<0 -333 333>)
   )
   &DynamicMaterialParameters(
    Index 0 Param 1=&MaskFloatBySpawnGroup(Masked Float=0.333, Passthrough Float=0.666)
    Module.Param0WriteEnabled=true
    Module.Param1WriteEnabled=false
    Module.Param2WriteEnabled=false
    Module.Param3WriteEnabled=false
   )
   &SkeletalMeshLocation(
    Apply Sampled Orientation to Mesh Particles=true
    Bone / Socket Index=$Particles.SpawnGroup
    Bone Sampling Mode=::NewEnumerator4
    Inherit Velocity=false
    Interpolate Mesh Sampling=false
    MaskSpawnGroup=false
    Sampled Position Offset=&MaskVectorBySpawnGroup(Masked Vector=<10 0 0>, Passthrough Vector=<-10 0 0>)
    Sampled Position Offset Coordinate Space:ENiagaraMeshTransforms=::NewEnumerator3
    Skeletal Mesh=$User.SkelMeshLimbs
   )
  ]
  particleUpdateScript [
   &InheritSourceMovement(Applied Position Scale=<0.4 0.4 1>, ApplyForce=false, ApplyPosition=true, ApplyPositionFalloff=false)
   &SolveForcesAndVelocity(Clamp Velocity=false, Limit Acceleration=false)
   &FN_UpdateRibbon(Detach Locked Particle=true, Enable Skeletal Position=true)
  ]
  rendererClasses [&NiagaraRibbonRendererProperties]
  renderers [{rendererClass=&NiagaraRibbonRendererProperties}]
 }
 Smoke_Upward {
  emitterUpdateScript [
   &EmitterState(Life Cycle Mode=::NewEnumerator1, Loop Duration=1, UseLoopDelay=false)
   &SpawnRate(
    SpawnRate=&MakeCustomFloatFromBool(Bool=$User.bIsFrontEndPreview, True Float=500, False Float=&Lerp_Float(A=80, B=0, Alpha=$System.Dive))
    Use Spawn Probability=true
   )
  ]
  particleSpawnScript [
   &FN_DivingAttributes_v2(SkeletalMesh:NiagaraDataInterfaceSkeletalMesh=$User.SkeletalMesh)
   &FN_SetSpawnGroup_v2(Number of Spawn Groups=4)
   /SetVariables(Particles.DiveCLAMP:NiagaraFloat=&ClampFloat(
     Float=&FN_FrontEndFloatModifier(Gameplay=$Particles.Dive, Front End=&Multiply_Float(A=$Particles.Dive, B=10))
     Min=-1
    )) {enabled=false, isSetParameters=true, module=SetVariables_#1}
   &InitializeParticle(
    Color=<rgba 0.1 0.1 0.1 0.5>
    Lifetime=&MakeCustomFloatFromBool(Bool=$User.bIsFrontEndPreview, True Float=0.4, False Float=1)
    Module.Write Color=true
    Module.Write Lifetime=true
    Module.Write Mass=false
    Module.Write Material Random=true
    Module.Write Position=false
    Module.Write Scale=false
    Module.Write SpriteRotation=false
    Module.Write SpriteSize=true
    Position=&SimulationPosition()
    Sprite Rotation=&UniformRangedFloat(Minimum=-20, Maximum=20)
    Sprite Size=&Vector2DFromFloat(Value=&RandomRangeFloat(Minimum=18, Maximum=26))
   )
   &DynamicMaterialParameters(
    Index 0 Param 2=&UniformRangedFloat(Minimum=0, Maximum=1)
    Index 0 Param 3=1
    Index 0 Param 4=&UniformRangedFloat(Minimum=1, Maximum=0.5)
    Module.Param0WriteEnabled=false
    Module.Param1WriteEnabled=true
    Module.Param2WriteEnabled=false
    Module.Param3WriteEnabled=true
   ) {enabled=false, module=DynamicMaterialParameters002}
   &AddVelocity(Velocity=&Lerp_Vector(
     A=&UniformRangedVector(Minimum=<-25 0 0>, Maximum=<25 0 0>)
     B=&UniformRangedVector(Minimum=<-50 0 0>, Maximum=<50 0 0>)
     Alpha=$Particles.DiveCLAMP
    )) {enabled=false}
   &AddVelocity(Velocity=&UniformRangedVector(Minimum=<0 0 500>, Maximum=<0 0 300>)) {module=AddVelocity001}
   &SkeletalMeshLocation(
    Bone / Socket Index=HlslExpression() {hlslExpression=((Particles.SpawnGroup*3)+2)}
    Bone Sampling Mode=::NewEnumerator4
    Inherit Velocity=false
    "Inherit Velocity (Scale)"=0
    Interpolate Mesh Sampling=false
    MaskSpawnGroup=false
    Skeletal Mesh=$User.SkeletalMesh
   )
   &CameraOffset(Camera Offset Amount=20)
   &SphereLocation(
    MaskSpawnGroup=false
    Override Seed=false
    Sphere Radius=5
    Surface Only=false
    UseNonUniformScale=false
    UseOffset=false
    UseSphereOrientation=false
   )
  ]
  particleUpdateScript [
   /SetVariables(Particles.DiveCLAMP:NiagaraFloat=&ClampFloat(Float=&FN_FrontEndFloatModifier(Gameplay=$Particles.Dive, Front End=&Multiply_Float(A=$Particles.Dive, B=10)))) {enabled=false, isSetParameters=true, module=SetVariables_#3}
   &ParticleState()
   &Drag(Drag=2, Use Linear Drag=true, Use Rotational Drag=false)
   &SolveForcesAndVelocity(Clamp Velocity=false, Limit Acceleration=false)
   &ScaleSpriteSize(Scale Factor=&Vector2DFromFloat(Value=&FloatFromCurve(FloatCurve=DataInterface() {propertyValues {Curve {keys [
          {arriveTangent=0.3786, interpMode=RCIM_Cubic, leaveTangent=0.3786, tangentMode=RCTM_User, time=0, value=1}
          {arriveTangent=3.33, interpMode=RCIM_Cubic, leaveTangent=3.33, tangentMode=RCTM_User, time=1, value=3}
         ]}}})))
   &ScaleColor(
    Scale Alpha=&FloatFromCurve(
     FloatCurve=DataInterface() {propertyValues {Curve {keys [
         {arriveTangent=8.7245, interpMode=RCIM_Cubic, leaveTangent=8.7245, tangentMode=RCTM_User, time=0, value=0}
         {arriveTangent=0.4876, interpMode=RCIM_Cubic, leaveTangent=0.4876, time=0.2, value=1}
         {arriveTangent=-1.25, interpMode=RCIM_Cubic, leaveTangent=-1.25, time=0.6306, value=0.3075}
         {arriveTangent=-0.5281, interpMode=RCIM_Cubic, leaveTangent=-0.5281, tangentMode=RCTM_User, time=1, value=0}
        ]}}}
     Scale Curve=&MakeCustomFloatFromBool(Bool=$User.bIsFrontEndPreview, True Float=0.3, False Float=1)
    )
    Scale RGB=&VectorFromFloat(Value=&FloatFromCurve(FloatCurve=DataInterface() {propertyValues {Curve {keys [
          {arriveTangent=0.0824, interpMode=RCIM_Cubic, leaveTangent=0.0824, tangentMode=RCTM_User, time=0, value=1}
          {arriveTangent=-2.0056, interpMode=RCIM_Cubic, leaveTangent=-2.0056, tangentMode=RCTM_User, time=1, value=0}
         ]}}}))
    ScaleA=true
    ScaleRGB=true
   )
   &ScaleColor(
    Scale RGB=&MakeVectorFromLinearColorRGB(Color=&ColorFromCurve(ColorCurve=DataInterface() {propertyValues {
        AlphaCurve {keys [<at 0 1 RCIM_Cubic>,<at 1 1 RCIM_Cubic>]}
        BlueCurve {keys [<at 0.0828 0.03>,<at 1 0.03 RCIM_Cubic>]}
        GreenCurve {keys [<at 0.0828 0.2887>,<at 1 0.0462 RCIM_Cubic>]}
        RedCurve {keys [<at 0.0828 1>,<at 1 1 RCIM_Cubic>]}
       }}))
    ScaleA=true
    ScaleRGB=true
   ) {enabled=false, module=ScaleColor001}
   &ScaleSpriteSize(Scale Factor=<1.05 1.05>) {enabled=false, module=ScaleSpriteSize001}
   &DynamicMaterialParameters(
    Index 0 Param 1=&Add_Float(
     A=&FloatFromCurve(FloatCurve=DataInterface() {propertyValues {Curve {keys [<at 0 0>,<at 2 1>]}}})
     B=&Multiply_Float(A=$Particles.Initial.MaterialRandom, B=0.5)
    )
    Index 0 Param 3=&FloatFromCurve(FloatCurve=DataInterface() {propertyValues {Curve {keys [
         {arriveTangent=-0.0944, interpMode=RCIM_Cubic, time=0, value=0}
         {arriveTangent=2.5569, interpMode=RCIM_Cubic, leaveTangent=2.5569, time=0.0519, value=2.3053}
         <at 0.5008 1.2804>
        ]}}})
    Index 0 Param 4=2
    Index 1 Param 3=1
    Index 2 Param 1=&FloatFromCurve(FloatCurve=DataInterface() {propertyValues {Curve {keys [
         {arriveTangent=0.8446, interpMode=RCIM_Cubic, leaveTangent=0.8446, tangentMode=RCTM_User, time=0, value=0.5}
         {arriveTangent=1.3293, interpMode=RCIM_Cubic, leaveTangent=1.3293, tangentMode=RCTM_User, time=1, value=2}
        ]}}})
    Index 2 Param 3=1
    Index 2 Param 4=1.5
    Module.Param0WriteEnabled=true
    Module.Param0WriteEnabled1=true
    Module.Param0WriteEnabled2=true
    Module.Param1WriteEnabled=true
    Module.Param1WriteEnabled1=true
    Module.Param1WriteEnabled2=true
    Module.Param2WriteEnabled=true
    Module.Param2WriteEnabled1=true
    Module.Param2WriteEnabled2=true
    Module.Param3WriteEnabled=true
    Module.Param3WriteEnabled1=true
    Module.Param3WriteEnabled2=true
    Write Parameter Index 1=true
    Write Parameter Index 2=true
   )
  ]
  rendererClasses [&NiagaraSpriteRendererProperties]
  renderers [{rendererClass=&NiagaraSpriteRendererProperties}]
 }
}
summary {
 systemName=NS_Contrail_Renegade
 emitters [
  {emitterName=Lights, rendererClasses [&NiagaraLightRendererProperties, &NiagaraSpriteRendererProperties]}
  {bEnabled=false, emitterName=Flames_FlowHands, rendererClasses [&NiagaraSpriteRendererProperties]}
  {bEnabled=false, emitterName=FN_TEST_AXIS, rendererClasses [&NiagaraMeshRendererProperties]}
  {emitterName=FN_FlameDecals, rendererClasses [&NiagaraSpriteRendererProperties]}
  {bEnabled=false, emitterName=Flames_FlowLegs, rendererClasses [&NiagaraSpriteRendererProperties]}
  {bEnabled=false, emitterName=CircleFlames, rendererClasses [&NiagaraSpriteRendererProperties]}
  {bEnabled=false, emitterName=FN_Contrails_Vertigo_Copied, rendererClasses [&NiagaraRibbonRendererProperties]}
  {bEnabled=false, emitterName=Flames_FlowHandsDark, rendererClasses [&NiagaraSpriteRendererProperties]}
  {bEnabled=false, emitterName=FN_Contrails_Ribbon_Small001, rendererClasses [&NiagaraRibbonRendererProperties]}
  {emitterName=Flames_Upward, rendererClasses [&NiagaraSpriteRendererProperties]}
  {bEnabled=false, emitterName=Mesh_Dive_Streaks, rendererClasses [&NiagaraMeshRendererProperties]}
  {bEnabled=false, emitterName=Blue_Main_Body, rendererClasses [&NiagaraMeshRendererProperties]}
  {bEnabled=false, emitterName=FN_AttachedRibbon_v2, rendererClasses [&NiagaraRibbonRendererProperties]}
  {bEnabled=false, emitterName=FN_SnowTrail_Ribbon_v4_A, rendererClasses [&NiagaraRibbonRendererProperties]}
  {emitterName=Smoke_Upward, rendererClasses [&NiagaraSpriteRendererProperties]}
  {emitterName=FireSmoke_Upward, rendererClasses [&NiagaraSpriteRendererProperties]}
  {bEnabled=false, emitterName=FN_AttachedRibbon_v2001, rendererClasses [&NiagaraRibbonRendererProperties]}
  {bEnabled=false, emitterName=Ribbons001, rendererClasses [&NiagaraRibbonRendererProperties]}
  {bEnabled=false, emitterName=Ribbons, rendererClasses [&NiagaraRibbonRendererProperties]}
  {emitterName=Flames_FlowLegs001, rendererClasses [&NiagaraSpriteRendererProperties]}
  {bEnabled=false, emitterName=Ribbon_FireStreaks, rendererClasses [&NiagaraRibbonRendererProperties]}
  {emitterName=Flames_FlowHands002, rendererClasses [&NiagaraSpriteRendererProperties]}
  {emitterName=Ribbon_Smoke, rendererClasses [&NiagaraRibbonRendererProperties]}
  {emitterName=Drops, simTarget=GPUComputeSim, rendererClasses [&NiagaraSpriteRendererProperties]}
 ]
 userVariables [
  User.AlphaInput {type=NiagaraFloat, defaultValue=0, description=""}
  User.bIsFrontEnd {type=NiagaraBool}
  User.bIsFrontEndPreview {type=NiagaraBool}
  User.FEVec {type=Vector3f, description="", defaultValue {struct=Vector3f, value=<0 0 0>}}
  User.FrontEndVector {type=Vector3f, description="", defaultValue {struct=Vector3f, value=<0 0 1000>}}
  User.PlayerSpeed {type=NiagaraFloat, defaultValue=0, description=""}
  User.RibbonIDOffset {type=NiagaraInt32}
  User.SkeletalMesh {
   type=NiagaraDataInterfaceSkeletalMesh
   defaultValue {
    struct=NiagaraExt_VariableValue_DataInterface
    value {dataInterface=&self:NiagaraDataInterfaceSkeletalMesh_0}
   }
  }
  User.SkelMeshLimbs {
   type=NiagaraDataInterfaceSkeletalMesh
   defaultValue {
    struct=NiagaraExt_VariableValue_DataInterface
    value {dataInterface=&self:NiagaraDataInterfaceSkeletalMesh_1}
   }
  }
 ]
}
